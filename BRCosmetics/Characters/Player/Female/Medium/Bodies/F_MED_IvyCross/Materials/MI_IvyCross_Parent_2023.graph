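# Parent skeletal-mesh material for the Ivy Cross female player body cosmetic, adding a crystal effect to the shared 2023 character shading.
# The character PreFX function supplies base material attributes and a tech art mask to the Crystal function borrowed from weapon skins.
# Channel-mask parameters select which mask channels drive the crystal's edge map and blend alpha.
# The result then passes through the shared PostFX function for final output.
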Material /BRCosmetics/Characters/Player/Female/Medium/Bodies/F_MED_IvyCross/Materials/MI_IvyCross_Parent_2023
graph {
 properties {
  bUseMaterialAttributes=true
  bUsedWithMeshDeformer=true
  bUsedWithMorphTargets=true
  bUsedWithSkeletalMesh=true
  pixelDepthOffsetMode=PDOM_Legacy
  usedShadingModels=Default Lit
 }
 nodes [
  n1 = MaterialFunctionCall() {materialFunction=&MF_CharacterParent_2023_PreFX}
  MaterialFunctionCall(
   Material=MaterialFunctionCall!(
    Edge Map=ChannelMaskParameter(None=@n1.Tech Art Mask) {group=05_Tech Art Masks, parameterName=Crystal Edge Channel}.R
    Material=@n1.Result
    Alpha=ChannelMaskParameter(None=@n1.Tech Art Mask) {group=05_Tech Art Masks, parameterName=Crystal Mask Channel}.R
   ) {materialFunction=/Game/Weapons/WeaponSkins/MaterialFunction/Crystal}.Result
   M=@n1.M
   Pre Skinned Local Position=@n1.M
   Distance Blend=@n1.M
   Fresnel=@n1.M
  ) {materialFunction=&MF_CharacterParent_2023_PostFX}
 ]
}
