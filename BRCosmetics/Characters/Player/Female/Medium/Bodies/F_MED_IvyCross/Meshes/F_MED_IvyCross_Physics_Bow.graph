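# Physics asset for the bow on the IvyCross medium female player body, covering the knot, bow loops and dangling front and back ties.
# Capsule chains branch from the knot, which hangs off the pelvis, into two short loops and, through a tie root, two longer tie strands.
# Chain joints have narrow swing and twist limits, mostly loosening toward the tie tips, while root joints swing freely.
# Pelvis, thigh and two front skirt capsules supply surrounding body shapes.
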PhysicsAsset /BRCosmetics/Characters/Player/Female/Medium/Bodies/F_MED_IvyCross/Meshes/F_MED_IvyCross_Physics_Bow
bodies {
 dyn_bow_bot_01 [{shapeType=Capsule, length=3, radius=1.5, shapeName=dyn_bow_bot_01_capsule, center=<1 0 0>}]
 dyn_bow_bot_02 [{shapeType=Capsule, length=3, radius=1.5, shapeName=dyn_bow_bot_02_capsule, center=<1 0 0>, rotation=<rot 0 0 90>}]
 dyn_bow_top_01 [{shapeType=Capsule, length=3, radius=1.5, shapeName=dyn_bow_top_01_capsule, center=<1 0 0>}]
 dyn_bow_top_02 [{shapeType=Capsule, length=3, radius=1.5, shapeName=dyn_bow_top_02_capsule, center=<3 0 0>, rotation=<rot 0 0 90>}]
 dyn_knot_MSTR [{
   shapeType=Capsule
   length=3.1168
   radius=2.2045
   shapeName=dyn_knot_MSTR_capsule
   center=<1.0178 -0.3151 0.2407>
   rotation=<rot 14.0615 -90 3.7765>
  }]
 dyn_skirt_fr_04 [{shapeType=Capsule, length=16, radius=3, shapeName=dyn_skirt_fr_04_capsule, center=<2.8025 0 0.4111>, rotation=<rot 0 0 90>}]
 dyn_skirt_r_fr_05 [{shapeType=Capsule, length=16, radius=3, shapeName=dyn_skirt_r_fr_05_capsule, center=<-0.1354 0 0.4111>, rotation=<rot -0.0001 0 105>}]
 dyn_tie_01 [{
   shapeType=Capsule
   length=2.8387
   radius=1.6065
   shapeName=dyn_tie_01_capsule
   center=<-0.4454 0.5068 -0.2752>
   rotation=<rot 0 78.4743 -80.0466>
  }]
 dyn_tie_MSTR [{shapeType=Capsule, length=0.505, radius=0.505, shapeName=dyn_tie_MSTR_capsule, rotation=<rot 0 -90 0>}]
 dyn_tie_bk_01 [{shapeType=Capsule, length=2.8821, radius=1.5, shapeName=dyn_tie_bk_01_capsule, center=<3 0 -1>, rotation=<rot 90 0 0>}]
 dyn_tie_bk_02 [{shapeType=Capsule, length=3.5, radius=1.5, shapeName=dyn_tie_bk_02_capsule, center=<3 0 -1>, rotation=<rot 90 0 0>}]
 dyn_tie_bk_03 [{shapeType=Capsule, length=8.75, radius=1.75, shapeName=dyn_tie_bk_03_capsule, center=<3 0 -1>, rotation=<rot 90 0 0>}]
 dyn_tie_bk_04 [{shapeType=Capsule, length=8.75, radius=1.75, shapeName=dyn_tie_bk_04_capsule, center=<2 0 -1>, rotation=<rot 90 0 0>}]
 dyn_tie_bk_05 [{shapeType=Capsule, length=8.75, radius=1.75, shapeName=dyn_tie_bk_05_capsule, center=<2 0 -1>, rotation=<rot 90 0 0>}]
 dyn_tie_fr_01 [{shapeType=Capsule, length=2.8821, radius=1.5, shapeName=dyn_tie_fr_01_capsule, center=<3 0 -1>, rotation=<rot 90 0 0>}]
 dyn_tie_fr_02 [{shapeType=Capsule, length=3.5, radius=1.5, shapeName=dyn_tie_fr_02_capsule, center=<3 0 -1>, rotation=<rot 90 0 0>}]
 dyn_tie_fr_03 [{shapeType=Capsule, length=8.75, radius=1.75, shapeName=dyn_tie_fr_03_capsule, center=<3 0 -1>, rotation=<rot 90 0 0>}]
 dyn_tie_fr_04 [{shapeType=Capsule, length=8.75, radius=1.75, shapeName=dyn_tie_fr_04_capsule, center=<3 0 -1>, rotation=<rot 90 0 0>}]
 dyn_tie_fr_05 [{shapeType=Capsule, length=8.75, radius=1.75, shapeName=dyn_tie_fr_05_capsule, center=<3 0 -1>, rotation=<rot 90 0 0>}]
 dyn_tie_fr_06 [{shapeType=Capsule, length=8.75, radius=1.75, shapeName=dyn_tie_fr_06_capsule, center=<3 0 -1>, rotation=<rot 90 0 0>}]
 pelvis [{
   shapeType=Capsule
   length=3.8723
   radius=10.0936
   shapeName=pelvis_capsule
   center=<6.5569 0.203 0.4387>
   rotation=<rot 0 -46.9189 -179.4136>
  }]
 thigh_l [{shapeType=Capsule, length=24.0377, radius=9.9309, center=<-8.2559 -1.008 4.04>, rotation=<rot 81.5625 179.9998 179.9999>}]
 thigh_r [{shapeType=Capsule, length=33.7679, radius=9.71, center=<16.9592 2.0447 -1.9176>, rotation=<rot 90 90 -89.9997>}]
}
constraints [
 {bone1Name=dyn_tie_bk_01, bone2Name=dyn_tie_01, swing1LimitDegrees=10, swing2LimitDegrees=10, twistLimitDegrees=2}
 {bone1Name=dyn_tie_bk_02, bone2Name=dyn_tie_bk_01, swing1LimitDegrees=10, swing2LimitDegrees=10, twistLimitDegrees=2}
 {bone1Name=dyn_tie_bk_03, bone2Name=dyn_tie_bk_02, swing1LimitDegrees=30, swing2LimitDegrees=30, twistLimitDegrees=3}
 {bone1Name=dyn_tie_bk_04, bone2Name=dyn_tie_bk_03, swing1LimitDegrees=30, swing2LimitDegrees=30, twistLimitDegrees=3}
 {bone1Name=dyn_tie_bk_05, bone2Name=dyn_tie_bk_04, swing1LimitDegrees=30, swing2LimitDegrees=30, twistLimitDegrees=3}
 {bone1Name=dyn_tie_fr_01, bone2Name=dyn_tie_01, swing1LimitDegrees=20, swing2LimitDegrees=20, twistLimitDegrees=2}
 {bone1Name=dyn_tie_fr_02, bone2Name=dyn_tie_fr_01, swing1LimitDegrees=20, swing2LimitDegrees=20, twistLimitDegrees=2}
 {bone1Name=dyn_tie_fr_03, bone2Name=dyn_tie_fr_02, swing1LimitDegrees=20, swing2LimitDegrees=20, twistLimitDegrees=2}
 {bone1Name=dyn_tie_fr_04, bone2Name=dyn_tie_fr_03, swing1LimitDegrees=30, swing2LimitDegrees=30, twistLimitDegrees=3}
 {bone1Name=dyn_tie_fr_05, bone2Name=dyn_tie_fr_04, swing1LimitDegrees=30, swing2LimitDegrees=30, twistLimitDegrees=3}
 {bone1Name=dyn_tie_fr_06, bone2Name=dyn_tie_fr_05, swing1LimitDegrees=30, swing2LimitDegrees=30, twistLimitDegrees=3}
 {bone1Name=dyn_bow_top_02, bone2Name=dyn_bow_top_01, swing1LimitDegrees=15, swing2LimitDegrees=15, twistLimitDegrees=2}
 {bone1Name=dyn_bow_bot_02, bone2Name=dyn_bow_bot_01, swing1LimitDegrees=15, swing2LimitDegrees=15, twistLimitDegrees=2}
 {bone1Name=dyn_bow_top_01, bone2Name=dyn_knot_MSTR, swing1LimitDegrees=15, swing2LimitDegrees=15, twistLimitDegrees=2}
 {bone1Name=dyn_bow_bot_01, bone2Name=dyn_knot_MSTR, swing1LimitDegrees=15, swing2LimitDegrees=15, twistLimitDegrees=2}
 {bone1Name=dyn_tie_01, bone2Name=dyn_tie_MSTR, swing1LimitDegrees=20, swing2LimitDegrees=15, twistLimitDegrees=1}
 {
  bone1Name=dyn_tie_MSTR
  bone2Name=dyn_knot_MSTR
  swing1LimitDegrees=45
  swing1Motion=Free
  swing2LimitDegrees=45
  swing2Motion=Free
  twistLimitDegrees=45
  twistMotion=Free
 }
 {
  bone1Name=dyn_knot_MSTR
  bone2Name=pelvis
  swing1LimitDegrees=45
  swing1Motion=Free
  swing2LimitDegrees=45
  swing2Motion=Free
  twistLimitDegrees=45
  twistMotion=Free
 }
 {
  bone1Name=thigh_l
  bone2Name=pelvis
  swing1LimitDegrees=45
  swing1Motion=Free
  swing2LimitDegrees=45
  swing2Motion=Free
  twistLimitDegrees=45
  twistMotion=Free
 }
 {
  bone1Name=thigh_r
  bone2Name=pelvis
  swing1LimitDegrees=45
  swing1Motion=Free
  swing2LimitDegrees=45
  swing2Motion=Free
  twistLimitDegrees=45
  twistMotion=Free
 }
 {
  bone1Name=dyn_skirt_fr_04
  bone2Name=pelvis
  swing1LimitDegrees=45
  swing1Motion=Free
  swing2LimitDegrees=45
  swing2Motion=Free
  twistLimitDegrees=45
  twistMotion=Free
 }
 {
  bone1Name=dyn_skirt_r_fr_05
  bone2Name=thigh_r
  swing1LimitDegrees=45
  swing1Motion=Free
  swing2LimitDegrees=45
  swing2Motion=Free
  twistLimitDegrees=45
  twistMotion=Free
 }
]
