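# Physics asset for the skirt of the female IvyCross character body, covering its dyn_skirt bones plus the pelvis and both thighs with capsules.
# Six four-bone chains (front, back, left, right, back-left, back-right) hang from the pelvis, with limited swing and very tight twist.
# Thigh-to-pelvis constraints leave rotation free.
# Free-rotation cross-constraints tie the lower two back-right segments to the center back chain, with no left-side equivalent.
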PhysicsAsset /BRCosmetics/Characters/Player/Female/Medium/Bodies/F_MED_IvyCross/Meshes/F_MED_IvyCross_Physics_Skirt
bodies {
 dyn_skirt_bk_01 [{shapeType=Capsule, length=8, radius=2, shapeName=dyn_skirt_bk_01_capsule, center=<3 0 -4>, rotation=<rot 0 0 90>}]
 dyn_skirt_bk_02 [{shapeType=Capsule, length=13.3129, radius=4.4432, shapeName=dyn_skirt_bk_02_capsule, center=<5 0 -2>, rotation=<rot 0 0 90>}]
 dyn_skirt_bk_03 [{shapeType=Capsule, length=18.0921, radius=3, shapeName=dyn_skirt_bk_03_capsule, center=<5 0 -2>, rotation=<rot 0 0 90>}]
 dyn_skirt_bk_04 [{shapeType=Capsule, length=18.0921, radius=3, shapeName=dyn_skirt_bk_04_capsule, center=<5 0 -2>, rotation=<rot 0 0 90>}]
 dyn_skirt_fr_01 [{shapeType=Capsule, length=10, radius=2.5, shapeName=dyn_skirt_fr_01_capsule, center=<3 0 -3>, rotation=<rot 0 0 90>}]
 dyn_skirt_fr_02 [{shapeType=Capsule, length=10, radius=2.5, shapeName=dyn_skirt_fr_02_capsule, center=<3 0 -3>, rotation=<rot 0 0 90>}]
 dyn_skirt_fr_03 [{shapeType=Capsule, length=10, radius=2.5, shapeName=dyn_skirt_fr_03_capsule, center=<3 0 -3>, rotation=<rot 0 0 90>}]
 dyn_skirt_fr_04 [{shapeType=Capsule, length=10, radius=2.5, shapeName=dyn_skirt_fr_04_capsule, center=<3 0 -3>, rotation=<rot 0 0 90>}]
 dyn_skirt_l_01 [{shapeType=Capsule, length=8, radius=1.5, shapeName=dyn_skirt_l_01_capsule, center=<3 0 -2>, rotation=<rot 0 0 90>}]
 dyn_skirt_l_02 [{shapeType=Capsule, length=10, radius=2.5, shapeName=dyn_skirt_l_02_capsule, center=<5 0 -1>, rotation=<rot 0 0 90>}]
 dyn_skirt_l_03 [{shapeType=Capsule, length=13.8459, radius=2.5, shapeName=dyn_skirt_l_03_capsule, center=<5 0 -1>, rotation=<rot 0 0 90>}]
 dyn_skirt_l_04 [{shapeType=Capsule, length=13.8459, radius=2.5, shapeName=dyn_skirt_l_04_capsule, center=<5 0 -1>, rotation=<rot 0 0 90>}]
 dyn_skirt_l_bk_01 [{shapeType=Capsule, length=8, radius=1.5, shapeName=dyn_skirt_l_bk_01_capsule, center=<3 0 -2>, rotation=<rot 0 0 90>}]
 dyn_skirt_l_bk_02 [{shapeType=Capsule, length=10, radius=3, shapeName=dyn_skirt_l_bk_02_capsule, center=<5 0 -2>, rotation=<rot 0 0 90>}]
 dyn_skirt_l_bk_03 [{shapeType=Capsule, length=17, radius=3.5, shapeName=dyn_skirt_l_bk_03_capsule, center=<5 0 -2>, rotation=<rot 0 0 90>}]
 dyn_skirt_l_bk_04 [{shapeType=Capsule, length=17, radius=3.5, shapeName=dyn_skirt_l_bk_04_capsule, center=<5 0 -2>, rotation=<rot 0 0 90>}]
 dyn_skirt_r_01 [{shapeType=Capsule, length=8, radius=2, shapeName=dyn_skirt_r_01_capsule, center=<3 0 -2>, rotation=<rot 0 0 90>}]
 dyn_skirt_r_02 [{shapeType=Capsule, length=10, radius=2.5, shapeName=dyn_skirt_r_02_capsule, center=<3 0 -2>, rotation=<rot 0 0 90>}]
 dyn_skirt_r_03 [{shapeType=Capsule, length=14.4574, radius=2.5, shapeName=dyn_skirt_r_03_capsule, center=<5 0 -2>, rotation=<rot 0 0 90>}]
 dyn_skirt_r_04 [{shapeType=Capsule, length=14.4574, radius=2.5, shapeName=dyn_skirt_r_04_capsule, center=<5 0 -2>, rotation=<rot 0 0 90>}]
 dyn_skirt_r_bk_01 [{shapeType=Capsule, length=8, radius=2, shapeName=dyn_skirt_r_bk_01_capsule, center=<3 0 -4>, rotation=<rot 0 0 90>}]
 dyn_skirt_r_bk_02 [{shapeType=Capsule, length=10, radius=3, shapeName=dyn_skirt_r_bk_02_capsule, center=<5 0 -2>, rotation=<rot 0 0 90>}]
 dyn_skirt_r_bk_03 [{shapeType=Capsule, length=17, radius=3.5, shapeName=dyn_skirt_r_bk_03_capsule, center=<5 0 -2>, rotation=<rot 0 0 90>}]
 dyn_skirt_r_bk_04 [{shapeType=Capsule, length=17, radius=3.5, shapeName=dyn_skirt_r_bk_04_capsule, center=<5 0 -2>, rotation=<rot 0 0 90>}]
 pelvis [{shapeType=Capsule, length=7.9903, radius=9.308, shapeName=pelvis_capsule, center=<-3.4742 0 0>, rotation=<rot 0 179.9997 -180>}]
 thigh_l [{shapeType=Capsule, length=23.9743, radius=8.3598, center=<-11.5655 -0.0001 1.5412>, rotation=<rot 84.375 179.9996 179.9998>}]
 thigh_r [{shapeType=Capsule, length=26.8528, radius=8.8408, center=<13.01 0.7023 -2.3444>, rotation=<rot 81.5625 179.9998 -0.0001>}]
}
constraints [
 {
  bone1Name=thigh_l
  bone2Name=pelvis
  swing1LimitDegrees=45
  swing1Motion=Free
  swing2LimitDegrees=45
  swing2Motion=Free
  twistLimitDegrees=45
  twistMotion=Free
 }
 {
  bone1Name=thigh_r
  bone2Name=pelvis
  swing1LimitDegrees=45
  swing1Motion=Free
  swing2LimitDegrees=45
  swing2Motion=Free
  twistLimitDegrees=45
  twistMotion=Free
 }
 {bone1Name=dyn_skirt_bk_01, bone2Name=pelvis, swing1LimitDegrees=30, swing2LimitDegrees=25, twistLimitDegrees=2}
 {bone1Name=dyn_skirt_bk_02, bone2Name=dyn_skirt_bk_01, swing1LimitDegrees=45, swing2LimitDegrees=30, twistLimitDegrees=2}
 {bone1Name=dyn_skirt_bk_03, bone2Name=dyn_skirt_bk_02, swing1LimitDegrees=45, swing2LimitDegrees=30, twistLimitDegrees=2}
 {bone1Name=dyn_skirt_bk_04, bone2Name=dyn_skirt_bk_03, swing1LimitDegrees=45, swing2LimitDegrees=30, twistLimitDegrees=2}
 {bone1Name=dyn_skirt_fr_02, bone2Name=dyn_skirt_fr_01, swing1LimitDegrees=20, swing2LimitDegrees=10, twistLimitDegrees=2}
 {bone1Name=dyn_skirt_fr_01, bone2Name=pelvis, swing1LimitDegrees=20, swing2LimitDegrees=10, twistLimitDegrees=2}
 {bone1Name=dyn_skirt_fr_03, bone2Name=dyn_skirt_fr_02, swing1LimitDegrees=20, swing2LimitDegrees=10, twistLimitDegrees=2}
 {bone1Name=dyn_skirt_fr_04, bone2Name=dyn_skirt_fr_03, swing1LimitDegrees=20, swing2LimitDegrees=10, twistLimitDegrees=2}
 {bone1Name=dyn_skirt_l_01, bone2Name=pelvis, swing1LimitDegrees=40, swing2LimitDegrees=40, twistLimitDegrees=2}
 {bone1Name=dyn_skirt_l_02, bone2Name=dyn_skirt_l_01, swing1LimitDegrees=40, swing2LimitDegrees=40, twistLimitDegrees=2}
 {bone1Name=dyn_skirt_l_03, bone2Name=dyn_skirt_l_02, swing1LimitDegrees=40, swing2LimitDegrees=40, twistLimitDegrees=2}
 {bone1Name=dyn_skirt_l_04, bone2Name=dyn_skirt_l_03, swing1LimitDegrees=40, swing2LimitDegrees=40, twistLimitDegrees=2}
 {bone1Name=dyn_skirt_l_bk_01, bone2Name=pelvis, swing1LimitDegrees=25, swing2LimitDegrees=45, twistLimitDegrees=2}
 {bone1Name=dyn_skirt_l_bk_02, bone2Name=dyn_skirt_l_bk_01, swing1LimitDegrees=45, swing2LimitDegrees=30, twistLimitDegrees=2}
 {bone1Name=dyn_skirt_l_bk_03, bone2Name=dyn_skirt_l_bk_02, swing1LimitDegrees=45, swing2LimitDegrees=30, twistLimitDegrees=2}
 {bone1Name=dyn_skirt_l_bk_04, bone2Name=dyn_skirt_l_bk_03, swing1LimitDegrees=45, swing2LimitDegrees=30, twistLimitDegrees=2}
 {bone1Name=dyn_skirt_r_01, bone2Name=pelvis, swing1LimitDegrees=20, swing2LimitDegrees=10, twistLimitDegrees=2}
 {bone1Name=dyn_skirt_r_02, bone2Name=dyn_skirt_r_01, swing1LimitDegrees=20, swing2LimitDegrees=10, twistLimitDegrees=2}
 {bone1Name=dyn_skirt_r_03, bone2Name=dyn_skirt_r_02, swing1LimitDegrees=20, swing2LimitDegrees=10, twistLimitDegrees=2}
 {bone1Name=dyn_skirt_r_04, bone2Name=dyn_skirt_r_03, swing1LimitDegrees=20, swing2LimitDegrees=10, twistLimitDegrees=2}
 {bone1Name=dyn_skirt_r_bk_01, bone2Name=pelvis, swing1LimitDegrees=25, swing2LimitDegrees=45, twistLimitDegrees=2}
 {bone1Name=dyn_skirt_r_bk_02, bone2Name=dyn_skirt_r_bk_01, swing1LimitDegrees=45, swing2LimitDegrees=30, twistLimitDegrees=2}
 {bone1Name=dyn_skirt_r_bk_03, bone2Name=dyn_skirt_r_bk_02, swing1LimitDegrees=45, swing2LimitDegrees=30, twistLimitDegrees=2}
 {bone1Name=dyn_skirt_r_bk_04, bone2Name=dyn_skirt_r_bk_03, swing1LimitDegrees=45, swing2LimitDegrees=30, twistLimitDegrees=2}
 {
  bone1Name=dyn_skirt_r_bk_03
  bone2Name=dyn_skirt_bk_03
  swing1LimitDegrees=45
  swing1Motion=Free
  swing2LimitDegrees=45
  swing2Motion=Free
  twistLimitDegrees=45
  twistMotion=Free
 }
 {
  bone1Name=dyn_skirt_r_bk_04
  bone2Name=dyn_skirt_bk_04
  swing1LimitDegrees=45
  swing1Motion=Free
  swing2LimitDegrees=45
  swing2Motion=Free
  twistLimitDegrees=45
  twistMotion=Free
 }
]
