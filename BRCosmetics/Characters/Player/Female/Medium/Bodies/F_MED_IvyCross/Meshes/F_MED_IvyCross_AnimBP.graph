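# Anim Blueprint for the F_MED_IvyCross female body cosmetic, adding secondary motion to skirt, ties, chains, thigh straps, gun and bow.
# It copies the parent mesh's pose and simulates skirt bones with RigidBody and AnimDynamics.
# Tie dynamics apply only while crouching, then straps and gun follow.
# Two Control Rigs take movement and action states, feeding bow RigidBody passes and final constraint and bone-offset corrections.
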AnimBlueprint /BRCosmetics/Characters/Player/Female/Medium/Bodies/F_MED_IvyCross/Meshes/F_MED_IvyCross_AnimBP
target_skeleton=&/F_MED_IvyCross_Skeleton
graphs [
 AnimGraph {nodes [
   Anim Dynamics - Bone: dyn_chain_fr_01!(ComponentPose, Alpha=1, GravityOverride:Vector="0,0,-980", ExternalForce="0,0,0") -> (Pose) {
    angularDampingOverride=0.99
    angularSpringConstant=200
    bAngularSpring=true
    bLinearSpring=true
    bOverrideAngularDamping=true
    bUseGravityOverride=true
    bUsePlanarLimit=false
    bUseSphericalLimits=true
    linearSpringConstant=50
    simulationSpace=BoneRelative
    alphaBoolBlend {blendOption=Linear}
    boundBone {boneName=dyn_chain_fr_01}
    componentAppliedLinearAccClamp=<3000 3000 3000>
    componentLinearAccScale=<0.5 0.3 0.3>
    componentLinearVelScale=<0.5 0.3 0.3>
    gravityOverride {z=-980}
    physicsBodyDefinitions [{
      boundBone {boneName=dyn_chain_fr_01}
      boxExtents=<20 10 10>
      constraintSetup {angularLimitsMax=<1 5 5>, angularLimitsMin=<-1 -10 -5>, angularTarget=<0 0 0>}
      localJointOffset=<4 0 0>
     }]
    relativeSpaceBone {boneName=root}
    sphericalLimits [{limitRadius=15, drivingBone {boneName=pelvis}, sphereLocalOffset=<-5 0 0>}]
   }
   Anim Dynamics - Bone: dyn_skirt_l_02!(ComponentPose, bAlphaBoolEnabled:bool=True, GravityOverride:Vector="0,0,-980", ExternalForce="0,0,0") -> (Pose) {
    alphaInputType=Bool
    angularBiasOverride=0.999
    angularDampingOverride=0.999
    angularSpringConstant=1500
    bOverrideAngularBias=true
    bOverrideAngularDamping=true
    bUseGravityOverride=true
    bUsePlanarLimit=false
    bUseSphericalLimits=true
    simulationSpace=RootRelative
    alphaBoolBlend {blendOption=Linear}
    boundBone {boneName=dyn_skirt_l_02}
    componentAppliedLinearAccClamp=<3000 3000 3000>
    componentLinearAccScale=<0.4 0.5 0.5>
    componentLinearVelScale=<0.4 0.5 0.5>
    physicsBodyDefinitions [{
      collisionType=CustomSphere
      sphereCollisionRadius=2
      boundBone {boneName=dyn_skirt_l_02}
      boxExtents=<7 4 4>
      constraintSetup {angularLimitsMax=<2.5 30 0>, angularLimitsMin=<-2.5 3 -10>, angularTarget=<0 0 0>}
      localJointOffset=<2 0 0>
     }]
    planarLimits [{drivingBone {boneName=spine_02}, planeTransform {location=<0 -20 0>, rotation=<rot 0 0 -90>}}]
    sphericalLimits [
     {limitRadius=7, drivingBone {boneName=pelvis}, sphereLocalOffset=<0 0 0>}
     {limitRadius=7, drivingBone {boneName=spine_01}, sphereLocalOffset=<5 0 0>}
     {limitRadius=7, drivingBone {boneName=spine_02}, sphereLocalOffset=<7 0 0>}
    ]
   }
   Anim Dynamics - Bone: dyn_skirt_r_02!(ComponentPose, bAlphaBoolEnabled:bool=True, GravityOverride:Vector="0,0,-980", ExternalForce="0,0,0") -> (Pose) {
    alphaInputType=Bool
    angularBiasOverride=0.999
    angularDampingOverride=0.999
    angularSpringConstant=1500
    bOverrideAngularBias=true
    bOverrideAngularDamping=true
    bUseGravityOverride=true
    bUsePlanarLimit=false
    bUseSphericalLimits=true
    simulationSpace=RootRelative
    alphaBoolBlend {blendOption=Linear}
    boundBone {boneName=dyn_skirt_r_02}
    componentAppliedLinearAccClamp=<3000 3000 3000>
    componentLinearAccScale=<0.4 0.5 0.5>
    componentLinearVelScale=<0.4 0.5 0.5>
    physicsBodyDefinitions [{
      collisionType=CustomSphere
      sphereCollisionRadius=2
      boundBone {boneName=dyn_skirt_r_02}
      boxExtents=<7 4 4>
      constraintSetup {angularLimitsMax=<2.5 30 0>, angularLimitsMin=<-2.5 3 -10>, angularTarget=<0 0 0>}
      localJointOffset=<2 0 0>
     }]
    planarLimits [{drivingBone {boneName=spine_02}, planeTransform {location=<0 -20 0>, rotation=<rot 0 0 -90>}}]
    sphericalLimits [
     {limitRadius=7, drivingBone {boneName=pelvis}, sphereLocalOffset=<0 0 0>}
     {limitRadius=7, drivingBone {boneName=spine_01}, sphereLocalOffset=<5 0 0>}
     {limitRadius=7, drivingBone {boneName=spine_02}, sphereLocalOffset=<7 0 0>}
    ]
   }
   "Transform (Modify) Bone - Bone: dyn_skirt_fr_MSTR"!(ComponentPose, bAlphaBoolEnabled:bool=True, Translation="0,0,0", Rotation="5.653,0,-4.758") -> (Pose) {
    alphaInputType=Bool
    rotationMode=BMM_Additive
    rotationSpace=BCS_ParentBoneSpace
    translationMode=BMM_Additive
    translationSpace=BCS_BoneSpace
    alphaBoolBlend {blendOption=Linear}
    boneToModify {boneName=dyn_skirt_fr_MSTR}
    rotation {pitch=5.653, roll=-4.758}
   }
   n1 = Anim Dynamics - Bone: dyn_tie_fr_03!(ComponentPose=Local To Component(LocalPose=Use cached pose 'Pose'()), Alpha=1, GravityOverride:Vector="0,0,-980", ExternalForce="0,0,0") -> (Pose) {
    angularBiasOverride=0.99
    angularDampingOverride=0.9999
    angularSpringConstant=500
    bAngularSpring=true
    bLinearSpring=true
    bOverrideAngularDamping=true
    bUseGravityOverride=true
    bUsePlanarLimit=false
    bUseSphericalLimits=true
    linearSpringConstant=50
    simulationSpace=RootRelative
    alphaBoolBlend {blendOption=Linear}
    boundBone {boneName=dyn_tie_fr_03}
    componentAppliedLinearAccClamp=<3000 3000 3000>
    componentLinearAccScale=<0.5 1 1>
    componentLinearVelScale=<0.5 1 1>
    physicsBodyDefinitions [{
      collisionType=CustomSphere
      sphereCollisionRadius=2
      boundBone {boneName=dyn_tie_fr_03}
      boxExtents=<4 7 4>
      constraintSetup {angularTargetAxis=AxisY, twistAxis=AxisY, angularLimitsMax=<2.5 30 20>, angularLimitsMin=<-2.5 -45 -45>, angularTarget=<0 0 0>}
      localJointOffset=<2 0 0>
     }]
    planarLimits [{drivingBone {boneName=spine_02}, planeTransform {location=<0 -20 0>, rotation=<rot 0 0 -90>}}]
    sphericalLimits [
     {limitRadius=7, drivingBone {boneName=pelvis}, sphereLocalOffset=<0 0 0>}
     {limitRadius=7, drivingBone {boneName=spine_01}, sphereLocalOffset=<5 0 0>}
     {limitRadius=7, drivingBone {boneName=spine_02}, sphereLocalOffset=<7 0 0>}
    ]
   }
   n2 = Anim Dynamics - Bone: dyn_tie_fr_04!(ComponentPose=@n1, Alpha=1, GravityOverride:Vector="0,0,-980", ExternalForce="0,0,0") -> (Pose) {
    angularBiasOverride=0.99
    angularDampingOverride=0.9999
    angularSpringConstant=500
    bAngularSpring=true
    bLinearSpring=true
    bOverrideAngularDamping=true
    bUseGravityOverride=true
    bUsePlanarLimit=false
    bUseSphericalLimits=true
    linearSpringConstant=50
    simulationSpace=RootRelative
    alphaBoolBlend {blendOption=Linear}
    boundBone {boneName=dyn_tie_fr_04}
    componentAppliedLinearAccClamp=<3000 3000 3000>
    componentLinearAccScale=<0.5 1 1>
    componentLinearVelScale=<0.5 1 1>
    physicsBodyDefinitions [{
      collisionType=CustomSphere
      sphereCollisionRadius=2
      boundBone {boneName=dyn_tie_fr_04}
      boxExtents=<5.5 10 5.5>
      constraintSetup {angularTargetAxis=AxisY, twistAxis=AxisY, angularLimitsMax=<2.5 30 20>, angularLimitsMin=<-2.5 -45 -45>, angularTarget=<0 0 0>}
      localJointOffset=<2 0 0>
     }]
    planarLimits [{drivingBone {boneName=spine_02}, planeTransform {location=<0 -20 0>, rotation=<rot 0 0 -90>}}]
   }
   n3 = Anim Dynamics - Bone: dyn_tie_fr_05!(ComponentPose=@n2, Alpha=1, GravityOverride:Vector="0,0,-980", ExternalForce="0,0,0") -> (Pose) {
    angularBiasOverride=0.99
    angularDampingOverride=0.9999
    angularSpringConstant=500
    bAngularSpring=true
    bLinearSpring=true
    bOverrideAngularDamping=true
    bUseGravityOverride=true
    bUsePlanarLimit=false
    bUseSphericalLimits=true
    linearSpringConstant=50
    simulationSpace=RootRelative
    alphaBoolBlend {blendOption=Linear}
    boundBone {boneName=dyn_tie_fr_05}
    componentLinearAccScale=<1 1 1>
    componentLinearVelScale=<1 1 1>
    physicsBodyDefinitions [{
      collisionType=CustomSphere
      sphereCollisionRadius=4
      boundBone {boneName=dyn_tie_fr_05}
      boxExtents=<8 17 8>
      constraintSetup {angularTargetAxis=AxisY, twistAxis=AxisY, angularLimitsMax=<2.5 30 20>, angularLimitsMin=<-2.5 -45 -45>, angularTarget=<0 0 0>}
      localJointOffset=<2 0 0>
     }]
    planarLimits [{drivingBone {boneName=spine_02}, planeTransform {location=<0 -20 0>, rotation=<rot 0 0 -90>}}]
   }
   n4 = Anim Dynamics - Bone: dyn_tie_fr_06!(ComponentPose=@n3, Alpha=1, GravityOverride:Vector="0,0,-980", ExternalForce="0,0,0") -> (Pose) {
    angularBiasOverride=0.99
    angularDampingOverride=0.9999
    angularSpringConstant=500
    bAngularSpring=true
    bLinearSpring=true
    bOverrideAngularDamping=true
    bUseGravityOverride=true
    bUsePlanarLimit=false
    bUseSphericalLimits=true
    linearSpringConstant=50
    simulationSpace=RootRelative
    alphaBoolBlend {blendOption=Linear}
    boundBone {boneName=dyn_tie_fr_06}
    componentLinearAccScale=<1 1 1>
    componentLinearVelScale=<1 1 1>
    physicsBodyDefinitions [{
      collisionType=CustomSphere
      sphereCollisionRadius=4
      boundBone {boneName=dyn_tie_fr_06}
      boxExtents=<8 17 8>
      constraintSetup {angularTargetAxis=AxisY, twistAxis=AxisY, angularLimitsMax=<2.5 30 20>, angularLimitsMin=<-2.5 -45 -45>, angularTarget=<0 0 0>}
      localJointOffset=<2 0 0>
     }]
    planarLimits [{drivingBone {boneName=spine_02}, planeTransform {location=<0 -20 0>, rotation=<rot 0 0 -90>}}]
   }
   n5 = Anim Dynamics - Bone: dyn_tie_bk_02!(ComponentPose=@n4, Alpha=1, GravityOverride:Vector="0,0,-980", ExternalForce="0,0,0") -> (Pose) {
    angularBiasOverride=0.99
    angularDampingOverride=0.9999
    angularSpringConstant=500
    bAngularSpring=true
    bLinearSpring=true
    bOverrideAngularDamping=true
    bUseGravityOverride=true
    bUsePlanarLimit=false
    bUseSphericalLimits=true
    linearSpringConstant=50
    simulationSpace=RootRelative
    alphaBoolBlend {blendOption=Linear}
    boundBone {boneName=dyn_tie_bk_02}
    componentAppliedLinearAccClamp=<3000 3000 3000>
    componentLinearAccScale=<0.5 1 1>
    componentLinearVelScale=<0.5 1 1>
    physicsBodyDefinitions [{
      collisionType=CustomSphere
      sphereCollisionRadius=2
      boundBone {boneName=dyn_tie_bk_02}
      boxExtents=<4 7 4>
      constraintSetup {angularTargetAxis=AxisY, twistAxis=AxisY, angularLimitsMax=<2.5 30 20>, angularLimitsMin=<-2.5 -45 -45>, angularTarget=<0 0 0>}
      localJointOffset=<2 0 0>
     }]
    planarLimits [{drivingBone {boneName=spine_02}, planeTransform {location=<0 -20 0>, rotation=<rot 0 0 -90>}}]
    sphericalLimits [
     {limitRadius=7, drivingBone {boneName=pelvis}, sphereLocalOffset=<0 0 0>}
     {limitRadius=7, drivingBone {boneName=spine_01}, sphereLocalOffset=<5 0 0>}
     {limitRadius=7, drivingBone {boneName=spine_02}, sphereLocalOffset=<7 0 0>}
    ]
   }
   n6 = Anim Dynamics - Bone: dyn_tie_bk_03!(ComponentPose=@n5, Alpha=1, GravityOverride:Vector="0,0,-980", ExternalForce="0,0,0") -> (Pose) {
    angularBiasOverride=0.99
    angularDampingOverride=0.9999
    angularSpringConstant=500
    bAngularSpring=true
    bLinearSpring=true
    bOverrideAngularDamping=true
    bUseGravityOverride=true
    bUsePlanarLimit=false
    bUseSphericalLimits=true
    linearSpringConstant=50
    simulationSpace=RootRelative
    alphaBoolBlend {blendOption=Linear}
    boundBone {boneName=dyn_tie_bk_03}
    componentAppliedLinearAccClamp=<3000 3000 3000>
    componentLinearAccScale=<0.5 1 1>
    componentLinearVelScale=<0.5 1 1>
    physicsBodyDefinitions [{
      collisionType=CustomSphere
      sphereCollisionRadius=2
      boundBone {boneName=dyn_tie_bk_03}
      boxExtents=<5.5 10 5.5>
      constraintSetup {angularTargetAxis=AxisY, twistAxis=AxisY, angularLimitsMax=<2.5 30 20>, angularLimitsMin=<-2.5 -45 -45>, angularTarget=<0 0 0>}
      localJointOffset=<2 0 0>
     }]
    planarLimits [{drivingBone {boneName=spine_02}, planeTransform {location=<0 -20 0>, rotation=<rot 0 0 -90>}}]
   }
   n7 = Anim Dynamics - Bone: dyn_tie_bk_04!(ComponentPose=@n6, Alpha=1, GravityOverride:Vector="0,0,-980", ExternalForce="0,0,0") -> (Pose) {
    angularBiasOverride=0.99
    angularDampingOverride=0.9999
    angularSpringConstant=500
    bAngularSpring=true
    bLinearSpring=true
    bOverrideAngularDamping=true
    bUseGravityOverride=true
    bUsePlanarLimit=false
    bUseSphericalLimits=true
    linearSpringConstant=50
    simulationSpace=RootRelative
    alphaBoolBlend {blendOption=Linear}
    boundBone {boneName=dyn_tie_bk_04}
    componentLinearAccScale=<1 1 1>
    componentLinearVelScale=<1 1 1>
    physicsBodyDefinitions [{
      collisionType=CustomSphere
      sphereCollisionRadius=4
      boundBone {boneName=dyn_tie_bk_04}
      boxExtents=<8 17 8>
      constraintSetup {angularTargetAxis=AxisY, twistAxis=AxisY, angularLimitsMax=<2.5 30 20>, angularLimitsMin=<-2.5 -45 -45>, angularTarget=<0 0 0>}
      localJointOffset=<2 0 0>
     }]
    planarLimits [{drivingBone {boneName=spine_02}, planeTransform {location=<0 -20 0>, rotation=<rot 0 0 -90>}}]
   }
   n8 = Anim Dynamics - Bone: dyn_tie_bk_05!(ComponentPose=@n7, Alpha=1, GravityOverride:Vector="0,0,-980", ExternalForce="0,0,0") -> (Pose) {
    angularBiasOverride=0.99
    angularDampingOverride=0.9999
    angularSpringConstant=500
    bAngularSpring=true
    bLinearSpring=true
    bOverrideAngularDamping=true
    bUseGravityOverride=true
    bUsePlanarLimit=false
    bUseSphericalLimits=true
    linearSpringConstant=50
    simulationSpace=RootRelative
    alphaBoolBlend {blendOption=Linear}
    boundBone {boneName=dyn_tie_bk_05}
    componentLinearAccScale=<1 1 1>
    componentLinearVelScale=<1 1 1>
    physicsBodyDefinitions [{
      collisionType=CustomSphere
      sphereCollisionRadius=4
      boundBone {boneName=dyn_tie_bk_05}
      boxExtents=<8 17 8>
      constraintSetup {angularTargetAxis=AxisY, twistAxis=AxisY, angularLimitsMax=<2.5 30 20>, angularLimitsMin=<-2.5 -45 -45>, angularTarget=<0 0 0>}
      localJointOffset=<2 0 0>
     }]
    planarLimits [{drivingBone {boneName=spine_02}, planeTransform {location=<0 -20 0>, rotation=<rot 0 0 -90>}}]
   }
   n9 = Component To Local(ComponentPose=@n8)
   n10 = Use cached pose 'Pose'()
   n11 = Get bIsCrouching()
   n12 = Anim Dynamics - Bone: dyn_thigh_strap_fr_01!(
    ComponentPose=Local To Component(LocalPose=Blend Poses by bool(BlendPose_0=@n9, BlendPose_1=@n10, bActiveValue=@n11) {blendType=Linear})
    Alpha=1
    GravityOverride:Vector="0,0,0"
    ExternalForce="0,0,0"
   ) -> (Pose) {
    angularDampingOverride=0.999
    angularSpringConstant=1000
    bAngularSpring=true
    bGravityOverrideInSimSpace=true
    bOverrideAngularDamping=true
    bUseGravityOverride=true
    simulationSpace=BoneRelative
    alphaBoolBlend {blendOption=Linear}
    boundBone {boneName=dyn_thigh_strap_fr_01}
    componentAppliedLinearAccClamp=<800 800 800>
    componentLinearAccScale=<1 1 1>
    componentLinearVelScale=<1 1 1>
    physicsBodyDefinitions [{
      sphereCollisionRadius=0
      boundBone {boneName=dyn_thigh_strap_fr_01}
      boxExtents=<18 22 18>
      constraintSetup {angularLimitsMax=<1 10 12>, angularLimitsMin=<-1 -10 -12>, angularTarget=<0 0 0>}
      localJointOffset=<-5 0 0>
     }]
    relativeSpaceBone {boneName=pelvis}
   }
   n13 = Anim Dynamics - Bone: dyn_thigh_strap_fr_02!(ComponentPose=@n12, Alpha=1, GravityOverride:Vector="0,0,0", ExternalForce="0,0,0") -> (Pose) {
    angularDampingOverride=0.999
    angularSpringConstant=1000
    bAngularSpring=true
    bGravityOverrideInSimSpace=true
    bOverrideAngularDamping=true
    bUseGravityOverride=true
    simulationSpace=BoneRelative
    alphaBoolBlend {blendOption=Linear}
    boundBone {boneName=dyn_thigh_strap_fr_02}
    componentAppliedLinearAccClamp=<800 800 800>
    componentLinearAccScale=<1 1 1>
    componentLinearVelScale=<1 1 1>
    physicsBodyDefinitions [{
      sphereCollisionRadius=0
      boundBone {boneName=dyn_thigh_strap_fr_02}
      boxExtents=<18 22 18>
      constraintSetup {angularLimitsMax=<1 10 12>, angularLimitsMin=<-1 -10 -12>, angularTarget=<0 0 0>}
      localJointOffset=<-5 0 0>
     }]
    relativeSpaceBone {boneName=pelvis}
   }
   n14 = Anim Dynamics - Bone: dyn_gun!(ComponentPose=@n13, Alpha=1, GravityOverride:Vector="0,0,0", ExternalForce="0,0,0") -> (Pose) {
    angularDampingOverride=0.999
    angularSpringConstant=1000
    bAngularSpring=true
    bGravityOverrideInSimSpace=true
    bOverrideAngularDamping=true
    bUseGravityOverride=true
    simulationSpace=BoneRelative
    alphaBoolBlend {blendOption=Linear}
    boundBone {boneName=dyn_gun}
    componentAppliedLinearAccClamp=<800 800 800>
    componentLinearAccScale=<1 1 1>
    componentLinearVelScale=<1 1 1>
    physicsBodyDefinitions [{
      sphereCollisionRadius=0
      boundBone {boneName=dyn_gun}
      boxExtents=<14 21 14>
      constraintSetup {angularLimitsMax=<12 12 0>, angularLimitsMin=<-12 -12 0>, angularTarget=<0 0 0>}
      localJointOffset=<-4.5 0 0>
     }]
    relativeSpaceBone {boneName=pelvis}
   }
   n15 = Anim Dynamics - Bone: dyn_thigh_strap_bk_01!(ComponentPose=@n14, Alpha=1, GravityOverride:Vector="0,0,0", ExternalForce="0,0,0") -> (Pose) {
    angularDampingOverride=0.999
    angularSpringConstant=1000
    bAngularSpring=true
    bGravityOverrideInSimSpace=true
    bOverrideAngularDamping=true
    bUseGravityOverride=true
    linearSpringConstant=1000
    simulationSpace=BoneRelative
    alphaBoolBlend {blendOption=Linear}
    boundBone {boneName=dyn_thigh_strap_bk_01}
    componentAppliedLinearAccClamp=<800 800 800>
    componentLinearAccScale=<1 1 1>
    componentLinearVelScale=<1 1 1>
    physicsBodyDefinitions [{
      sphereCollisionRadius=0
      boundBone {boneName=dyn_thigh_strap_bk_01}
      boxExtents=<14 16 14>
      constraintSetup {angularLimitsMax=<1 8 12>, angularLimitsMin=<-1 -8 -12>, angularTarget=<0 0 0>}
      localJointOffset=<-5 0 0>
     }]
    relativeSpaceBone {boneName=pelvis}
   }
   n16 = Anim Dynamics - Bone: dyn_thigh_strap_bk_02!(ComponentPose=@n15, Alpha=1, GravityOverride:Vector="0,0,0", ExternalForce="0,0,0") -> (Pose) {
    angularDampingOverride=0.999
    angularSpringConstant=1000
    bAngularSpring=true
    bGravityOverrideInSimSpace=true
    bOverrideAngularDamping=true
    bUseGravityOverride=true
    linearSpringConstant=1000
    simulationSpace=BoneRelative
    alphaBoolBlend {blendOption=Linear}
    boundBone {boneName=dyn_thigh_strap_bk_02}
    componentAppliedLinearAccClamp=<800 800 800>
    componentLinearAccScale=<1 1 1>
    componentLinearVelScale=<1 1 1>
    physicsBodyDefinitions [{
      sphereCollisionRadius=0
      boundBone {boneName=dyn_thigh_strap_bk_02}
      boxExtents=<14 16 14>
      constraintSetup {angularLimitsMax=<1 8 8>, angularLimitsMin=<-1 -8 -8>, angularTarget=<0 0 0>}
      localJointOffset=<-5 0 0>
     }]
    relativeSpaceBone {boneName=pelvis}
   }
   n17 = "Constraint - dyn_knot_MSTR (PPPPP)"!(
    ComponentPose=@n16
    Alpha=1
    ConstraintWeights_0=0.223
    ConstraintWeights_1:float=0.053
    ConstraintWeights_2:float=0.0067
    ConstraintWeights_3:float=0.1341
    ConstraintWeights_4:float=0.5832
   ) -> (Pose) {
    alphaBoolBlend {blendOption=Linear}
    boneToModify {boneName=dyn_knot_MSTR}
    constraintSetup [
     {targetBone {boneName=spine_02}}
     {targetBone {boneName=spine_04}}
     {targetBone {boneName=spine_03}}
     {targetBone {boneName=spine_01}}
     {targetBone {boneName=pelvis}}
    ]
    constraintWeights [1,1,1,1,1]
   }
   n18 = Control Rig!(
    Source=Component To Local(ComponentPose=@n17)
    Alpha
    Skydive:bool=Get bIsSkyDiving()
    Emote:bool=Get bIsPlayingEmote()
    Jump:bool=Get Jumps() -> (Jumps:bool)
    Slide BaseBall:bool=Get BaseBall_Slide() -> (BaseBall_Slide:bool)
    Slide Knee:bool=Get Knee_Slide() -> (Knee_Slide:bool)
   ) -> (Pose) {
    alphaBoolBlend {blendOption=Linear}
    controlRigAssetReference {blueprintRigClass=&/F_MED_IvyCross_CtrlRig_post.F_MED_IvyCross_CtrlRig_post_C}
   }
   n19 = Control Rig!(
    Source=@n18
    Alpha
    Crouch:bool=Get bIsCrouching()
    Left:double=Get Left() -> (Left:double)
    Right:double=Get Right() -> (Right:double)
    Forward:double=Get Forward() -> (Forward:double)
    Backward:double=Get Backward() -> (Backward:double)
    Moving:bool=Get bIsMoving2D()
    Targeting:bool=Get bIsTargeting()
    Melee:double=Get melee() -> (melee:double)
    Emote=Get bIsPlayingEmote()
    Skydive=Get bIsSkyDiving()
    DBNO:bool=Get bIsDBNO()
   ) -> (Pose) {
    alphaBoolBlend {blendOption=Linear}
    controlRigAssetReference {blueprintRigClass=&/F_MED_IvyCross_CtrlRig.F_MED_IvyCross_CtrlRig_C}
   }
   n20 = RigidBody!(ComponentPose=Local To Component(LocalPose=@n19), Alpha, OverrideWorldGravity:Vector="0,0,-980") -> (Pose) {
    bClampLinearTranslationLimitToRefPose=true
    bOverrideWorldGravity=true
    overridePhysicsAsset=&/F_MED_IvyCross_Physics_Bow
    simulationSpace=BaseBoneSpace
    alphaBoolBlend {blendInTime=0.5, blendOption=Linear, blendOutTime=0.5}
    baseBoneRef {boneName=pelvis}
    componentAppliedLinearAccClamp=<3000 3000 3000>
    componentLinearAccScale=<0.25 0.25 0.25>
    componentLinearVelScale=<0.3 0.3 0.3>
    overrideWorldGravity {z=-980}
    simSpaceSettings {
     maxAngularAcceleration=50
     maxAngularVelocity=50
     maxLinearAcceleration=100
     maxLinearVelocity=100
     externalLinearDragV {x=20}
     externalLinearVelocity=<50 50 50>
    }
   }
   n21 = RigidBody(ComponentPose=@n20, Alpha=Get Crouch_off() -> (Crouch_off:double), OverrideWorldGravity="0,0,-980") {
    bClampLinearTranslationLimitToRefPose=true
    bOverrideWorldGravity=true
    overridePhysicsAsset=&/F_MED_IvyCross_Physics_Bow_2
    simulationSpace=BaseBoneSpace
    alphaBoolBlend {blendInTime=0.5, blendOption=Linear, blendOutTime=0.5}
    baseBoneRef {boneName=pelvis}
    componentAppliedLinearAccClamp=<3000 3000 3000>
    componentLinearAccScale=<0.25 0.25 0.25>
    componentLinearVelScale=<0.3 0.3 0.3>
    overrideWorldGravity {z=-980}
    simSpaceSettings {
     maxAngularAcceleration=50
     maxAngularVelocity=50
     maxLinearAcceleration=100
     maxLinearVelocity=100
     externalLinearDragV {x=20}
     externalLinearVelocity=<50 50 50>
    }
   }
   n22 = "Constraint - dyn_chain_mid_02 (PPPPPPPP)"!(
    ComponentPose=@n21
    Alpha=1
    ConstraintWeights_0=0.1153
    ConstraintWeights_1:float=0.0626
    ConstraintWeights_2:float=0.004
    ConstraintWeights_3:float=0.096
    ConstraintWeights_4:float=0.3554
    ConstraintWeights_5:float=0.2317
    ConstraintWeights_6:float=0.013
    ConstraintWeights_7:float=0.122
   ) -> (Pose) {
    alphaBoolBlend {blendOption=Linear}
    boneToModify {boneName=dyn_chain_mid_02}
    constraintSetup [
     {targetBone {boneName=dyn_skirt_l_02}}
     {targetBone {boneName=dyn_skirt_l_03}}
     {targetBone {boneName=dyn_skirt_l_04}}
     {targetBone {boneName=dyn_skirt_l_01}}
     {targetBone {boneName=dyn_skirt_l_bk_02}}
     {targetBone {boneName=dyn_skirt_l_bk_03}}
     {targetBone {boneName=dyn_skirt_l_bk_04}}
     {targetBone {boneName=dyn_skirt_l_bk_01}}
    ]
    constraintWeights [1,1,1,1,1,1,1,1]
   }
   n23 = "Transform (Modify) Bone - Bone: dyn_chain_mid_02"!(ComponentPose=@n22, bAlphaBoolEnabled:bool=True, Translation="0.000,-0.033,-1.665", Rotation="4.986,0.439,5.023") -> (Pose) {
    alphaInputType=Bool
    rotationMode=BMM_Additive
    rotationSpace=BCS_BoneSpace
    translationMode=BMM_Additive
    translationSpace=BCS_BoneSpace
    alphaBoolBlend {blendOption=Linear}
    boneToModify {boneName=dyn_chain_mid_02}
    rotation=<rot 4.986 0.439 5.023>
    translation {y=-0.033, z=-1.665}
   }
   n24 = "Transform (Modify) Bone - Bone: dyn_chain_bk_MSTR"!(ComponentPose=@n23, bAlphaBoolEnabled:bool=True, Translation="0,0,0", Rotation="-10,0,0") -> (Pose) {
    alphaInputType=Bool
    rotationMode=BMM_Additive
    rotationSpace=BCS_ParentBoneSpace
    translationMode=BMM_Additive
    translationSpace=BCS_BoneSpace
    alphaBoolBlend {blendOption=Linear}
    boneToModify {boneName=dyn_chain_bk_MSTR}
    rotation {pitch=-10}
   }
   n25 = "Transform (Modify) Bone - Bone: dyn_chain_fr_MSTR"!(ComponentPose=@n24, bAlphaBoolEnabled:bool=True, Translation="0,0,0", Rotation="-9.218,-2.164,-1.438") -> (Pose) {
    alphaInputType=Bool
    rotationMode=BMM_Additive
    rotationSpace=BCS_ParentBoneSpace
    translationMode=BMM_Additive
    translationSpace=BCS_BoneSpace
    alphaBoolBlend {blendOption=Linear}
    boneToModify {boneName=dyn_chain_fr_MSTR}
    rotation=<rot -9.218 -2.164 -1.438>
   }
   n26 = "Constraint - dyn_skirt_fr_02 (PP)"!(ComponentPose=@n25, Alpha=0.4, ConstraintWeights_0=0.5, ConstraintWeights_1:float=0.5) -> (Pose) {
    alphaBoolBlend {blendOption=Linear}
    alphaScaleBiasClamp {bClampResult=true, clampMax=0.4}
    boneToModify {boneName=dyn_skirt_fr_02}
    constraintSetup [
     {perAxis {bY=false, bZ=false}, targetBone {boneName=dyn_skirt_r_fr_03}}
     {perAxis {bY=false, bZ=false}, targetBone {boneName=dyn_skirt_l_fr_03}}
    ]
    constraintWeights [1,1]
   }
   Output Pose(Result=Component To Local(ComponentPose=@n26)) {layerGroup=None}
   Save cached pose 'Copy'(Pose=Copy Pose From Mesh() {bUseAttachedParent=true}) {cacheName=Copy}
   n27 = RigidBody(ComponentPose=Local To Component(LocalPose=Use cached pose 'Copy'()), OverrideWorldGravity="0,0,-980") {
    bClampLinearTranslationLimitToRefPose=true
    bOverrideWorldGravity=true
    overridePhysicsAsset=&/F_MED_IvyCross_Physics_Skirt
    simulationSpace=BaseBoneSpace
    alphaBoolBlend {blendInTime=0.5, blendOption=Linear, blendOutTime=0.5}
    baseBoneRef {boneName=pelvis}
    componentAppliedLinearAccClamp=<3000 3000 3000>
    componentLinearAccScale=<0.25 0.25 0.25>
    componentLinearVelScale=<0.3 0.3 0.3>
    overrideWorldGravity {z=-980}
    simSpaceSettings {
     maxAngularAcceleration=50
     maxAngularVelocity=50
     maxLinearAcceleration=100
     maxLinearVelocity=100
     externalLinearDragV {x=20}
     externalLinearVelocity=<50 50 50>
    }
   }
   n28 = Anim Dynamics - Bone: dyn_skirt_l_fr_02!(ComponentPose=@n27, bAlphaBoolEnabled:bool=True, GravityOverride:Vector="0,0,-980", ExternalForce="0,0,0") -> (Pose) {
    alphaInputType=Bool
    angularBiasOverride=0.999
    angularDampingOverride=0.999
    angularSpringConstant=1500
    bOverrideAngularBias=true
    bOverrideAngularDamping=true
    bUseGravityOverride=true
    bUsePlanarLimit=false
    bUseSphericalLimits=true
    simulationSpace=RootRelative
    alphaBoolBlend {blendOption=Linear}
    boundBone {boneName=dyn_skirt_l_fr_02}
    componentAppliedLinearAccClamp=<3000 3000 3000>
    componentLinearAccScale=<0.4 0.5 0.5>
    componentLinearVelScale=<0.4 0.5 0.5>
    physicsBodyDefinitions [{
      collisionType=CustomSphere
      sphereCollisionRadius=2
      boundBone {boneName=dyn_skirt_l_fr_02}
      boxExtents=<7 4 4>
      constraintSetup {angularLimitsMax=<2.5 30 0>, angularLimitsMin=<-2.5 3 -10>, angularTarget=<0 0 0>}
      localJointOffset=<2 0 0>
     }]
    planarLimits [{drivingBone {boneName=spine_02}, planeTransform {location=<0 -20 0>, rotation=<rot 0 0 -90>}}]
   }
   n29 = Anim Dynamics - Bone: dyn_skirt_l_fr_03!(ComponentPose=@n28, bAlphaBoolEnabled:bool=True, GravityOverride:Vector="0,0,-980", ExternalForce="0,0,0") -> (Pose) {
    alphaInputType=Bool
    angularBiasOverride=0.999
    angularDampingOverride=0.999
    angularSpringConstant=1500
    bOverrideAngularBias=true
    bOverrideAngularDamping=true
    bUseGravityOverride=true
    bUsePlanarLimit=false
    bUseSphericalLimits=true
    simulationSpace=RootRelative
    alphaBoolBlend {blendOption=Linear}
    boundBone {boneName=dyn_skirt_l_fr_03}
    componentAppliedLinearAccClamp=<3000 3000 3000>
    componentLinearAccScale=<0.4 0.5 0.5>
    componentLinearVelScale=<0.4 0.5 0.5>
    physicsBodyDefinitions [{
      collisionType=CustomSphere
      sphereCollisionRadius=2
      boundBone {boneName=dyn_skirt_l_fr_03}
      boxExtents=<10 5.5 5.5>
      constraintSetup {angularLimitsMax=<2.5 20 0>, angularLimitsMin=<-2.5 -20 -20>, angularTarget=<0 0 0>}
      localJointOffset=<2 0 0>
     }]
    planarLimits [{drivingBone {boneName=spine_02}, planeTransform {location=<0 -20 0>, rotation=<rot 0 0 -90>}}]
   }
   n30 = Anim Dynamics - Bone: dyn_skirt_l_fr_04!(ComponentPose=@n29, bAlphaBoolEnabled:bool=True, GravityOverride:Vector="0,0,-980", ExternalForce="0,0,0") -> (Pose) {
    alphaInputType=Bool
    angularBiasOverride=0.999
    angularDampingOverride=0.999
    angularSpringConstant=1500
    bOverrideAngularBias=true
    bOverrideAngularDamping=true
    bUseGravityOverride=true
    bUsePlanarLimit=false
    bUseSphericalLimits=true
    simulationSpace=RootRelative
    alphaBoolBlend {blendOption=Linear}
    boundBone {boneName=dyn_skirt_l_fr_04}
    componentAppliedLinearAccClamp=<3000 3000 3000>
    componentLinearAccScale=<0.4 0.5 0.5>
    componentLinearVelScale=<0.4 0.5 0.5>
    physicsBodyDefinitions [{
      collisionType=CustomSphere
      sphereCollisionRadius=4
      boundBone {boneName=dyn_skirt_l_fr_04}
      boxExtents=<17 8 8>
      constraintSetup {angularLimitsMax=<2.5 10 0>, angularLimitsMin=<-2.5 -10 -20>, angularTarget=<0 0 0>}
      localJointOffset=<2 0 0>
     }]
    planarLimits [{drivingBone {boneName=spine_02}, planeTransform {location=<0 -20 0>, rotation=<rot 0 0 -90>}}]
   }
   n31 = Anim Dynamics - Bone: dyn_skirt_l_03!(ComponentPose=@n30, bAlphaBoolEnabled:bool=True, GravityOverride:Vector="0,0,-980", ExternalForce="0,0,0") -> (Pose) {
    alphaInputType=Bool
    angularBiasOverride=0.999
    angularDampingOverride=0.999
    angularSpringConstant=1500
    bOverrideAngularBias=true
    bOverrideAngularDamping=true
    bUseGravityOverride=true
    bUsePlanarLimit=false
    bUseSphericalLimits=true
    simulationSpace=RootRelative
    alphaBoolBlend {blendOption=Linear}
    boundBone {boneName=dyn_skirt_l_03}
    componentAppliedLinearAccClamp=<3000 3000 3000>
    componentLinearAccScale=<0.4 0.5 0.5>
    componentLinearVelScale=<0.4 0.5 0.5>
    physicsBodyDefinitions [{
      collisionType=CustomSphere
      sphereCollisionRadius=2
      boundBone {boneName=dyn_skirt_l_03}
      boxExtents=<10 5.5 5.5>
      constraintSetup {angularLimitsMax=<2.5 20 0>, angularLimitsMin=<-2.5 -20 -20>, angularTarget=<0 0 0>}
      localJointOffset=<2 0 0>
     }]
    planarLimits [{drivingBone {boneName=spine_02}, planeTransform {location=<0 -20 0>, rotation=<rot 0 0 -90>}}]
   }
   n32 = Anim Dynamics - Bone: dyn_skirt_l_04!(ComponentPose=@n31, bAlphaBoolEnabled:bool=True, GravityOverride:Vector="0,0,-980", ExternalForce="0,0,0") -> (Pose) {
    alphaInputType=Bool
    angularBiasOverride=0.999
    angularDampingOverride=0.999
    angularSpringConstant=1500
    bOverrideAngularBias=true
    bOverrideAngularDamping=true
    bUseGravityOverride=true
    bUsePlanarLimit=false
    bUseSphericalLimits=true
    simulationSpace=RootRelative
    alphaBoolBlend {blendOption=Linear}
    boundBone {boneName=dyn_skirt_l_04}
    componentAppliedLinearAccClamp=<3000 3000 3000>
    componentLinearAccScale=<0.4 0.5 0.5>
    componentLinearVelScale=<0.4 0.5 0.5>
    physicsBodyDefinitions [{
      collisionType=CustomSphere
      sphereCollisionRadius=4
      boundBone {boneName=dyn_skirt_l_04}
      boxExtents=<17 8 8>
      constraintSetup {angularLimitsMax=<2.5 10 0>, angularLimitsMin=<-2.5 -10 -20>, angularTarget=<0 0 0>}
      localJointOffset=<2 0 0>
     }]
    planarLimits [{drivingBone {boneName=spine_02}, planeTransform {location=<0 -20 0>, rotation=<rot 0 0 -90>}}]
   }
   n33 = Anim Dynamics - Bone: dyn_skirt_r_fr_03!(ComponentPose=@n32, bAlphaBoolEnabled:bool=True, GravityOverride:Vector="0,0,-980", ExternalForce="0,0,0") -> (Pose) {
    alphaInputType=Bool
    angularBiasOverride=0.999
    angularDampingOverride=0.999
    angularSpringConstant=1500
    bOverrideAngularBias=true
    bOverrideAngularDamping=true
    bUseGravityOverride=true
    bUsePlanarLimit=false
    bUseSphericalLimits=true
    simulationSpace=RootRelative
    alphaBoolBlend {blendOption=Linear}
    boundBone {boneName=dyn_skirt_r_fr_03}
    componentAppliedLinearAccClamp=<3000 3000 3000>
    componentLinearAccScale=<0.4 0.5 0.5>
    componentLinearVelScale=<0.4 0.5 0.5>
    physicsBodyDefinitions [{
      collisionType=CustomSphere
      sphereCollisionRadius=2
      boundBone {boneName=dyn_skirt_r_fr_03}
      boxExtents=<7 4 4>
      constraintSetup {angularLimitsMax=<2.5 30 0>, angularLimitsMin=<-2.5 3 -10>, angularTarget=<0 0 0>}
      localJointOffset=<2 0 0>
     }]
    planarLimits [{drivingBone {boneName=spine_02}, planeTransform {location=<0 -20 0>, rotation=<rot 0 0 -90>}}]
   }
   n34 = Anim Dynamics - Bone: dyn_skirt_r_fr_04!(ComponentPose=@n33, bAlphaBoolEnabled:bool=True, GravityOverride:Vector="0,0,-980", ExternalForce="0,0,0") -> (Pose) {
    alphaInputType=Bool
    angularBiasOverride=0.999
    angularDampingOverride=0.999
    angularSpringConstant=1500
    bOverrideAngularBias=true
    bOverrideAngularDamping=true
    bUseGravityOverride=true
    bUsePlanarLimit=false
    bUseSphericalLimits=true
    simulationSpace=RootRelative
    alphaBoolBlend {blendOption=Linear}
    boundBone {boneName=dyn_skirt_r_fr_04}
    componentAppliedLinearAccClamp=<3000 3000 3000>
    componentLinearAccScale=<0.4 0.5 0.5>
    componentLinearVelScale=<0.4 0.5 0.5>
    physicsBodyDefinitions [{
      collisionType=CustomSphere
      sphereCollisionRadius=2
      boundBone {boneName=dyn_skirt_r_fr_04}
      boxExtents=<10 5.5 5.5>
      constraintSetup {angularLimitsMax=<2.5 20 0>, angularLimitsMin=<-2.5 -20 -20>, angularTarget=<0 0 0>}
      localJointOffset=<2 0 0>
     }]
    planarLimits [{drivingBone {boneName=spine_02}, planeTransform {location=<0 -20 0>, rotation=<rot 0 0 -90>}}]
   }
   n35 = Anim Dynamics - Bone: dyn_skirt_r_fr_05!(ComponentPose=@n34, bAlphaBoolEnabled:bool=True, GravityOverride:Vector="0,0,-980", ExternalForce="0,0,0") -> (Pose) {
    alphaInputType=Bool
    angularBiasOverride=0.999
    angularDampingOverride=0.999
    angularSpringConstant=1500
    bOverrideAngularBias=true
    bOverrideAngularDamping=true
    bUseGravityOverride=true
    bUsePlanarLimit=false
    bUseSphericalLimits=true
    simulationSpace=RootRelative
    alphaBoolBlend {blendOption=Linear}
    boundBone {boneName=dyn_skirt_r_fr_05}
    componentAppliedLinearAccClamp=<3000 3000 3000>
    componentLinearAccScale=<0.4 0.5 0.5>
    componentLinearVelScale=<0.4 0.5 0.5>
    physicsBodyDefinitions [{
      collisionType=CustomSphere
      sphereCollisionRadius=4
      boundBone {boneName=dyn_skirt_r_fr_05}
      boxExtents=<17 8 8>
      constraintSetup {angularLimitsMax=<2.5 10 0>, angularLimitsMin=<-2.5 -10 -20>, angularTarget=<0 0 0>}
      localJointOffset=<2 0 0>
     }]
    planarLimits [{drivingBone {boneName=spine_02}, planeTransform {location=<0 -20 0>, rotation=<rot 0 0 -90>}}]
   }
   n36 = Anim Dynamics - Bone: dyn_skirt_r_03!(ComponentPose=@n35, bAlphaBoolEnabled:bool=True, GravityOverride:Vector="0,0,-980", ExternalForce="0,0,0") -> (Pose) {
    alphaInputType=Bool
    angularBiasOverride=0.999
    angularDampingOverride=0.999
    angularSpringConstant=1500
    bOverrideAngularBias=true
    bOverrideAngularDamping=true
    bUseGravityOverride=true
    bUsePlanarLimit=false
    bUseSphericalLimits=true
    simulationSpace=RootRelative
    alphaBoolBlend {blendOption=Linear}
    boundBone {boneName=dyn_skirt_r_03}
    componentAppliedLinearAccClamp=<3000 3000 3000>
    componentLinearAccScale=<0.4 0.5 0.5>
    componentLinearVelScale=<0.4 0.5 0.5>
    physicsBodyDefinitions [{
      collisionType=CustomSphere
      sphereCollisionRadius=2
      boundBone {boneName=dyn_skirt_r_03}
      boxExtents=<10 5.5 5.5>
      constraintSetup {angularLimitsMax=<2.5 20 0>, angularLimitsMin=<-2.5 -20 -20>, angularTarget=<0 0 0>}
      localJointOffset=<2 0 0>
     }]
    planarLimits [{drivingBone {boneName=spine_02}, planeTransform {location=<0 -20 0>, rotation=<rot 0 0 -90>}}]
   }
   n37 = Anim Dynamics - Bone: dyn_skirt_r_04!(ComponentPose=@n36, bAlphaBoolEnabled:bool=True, GravityOverride:Vector="0,0,-980", ExternalForce="0,0,0") -> (Pose) {
    alphaInputType=Bool
    angularBiasOverride=0.999
    angularDampingOverride=0.999
    angularSpringConstant=1500
    bOverrideAngularBias=true
    bOverrideAngularDamping=true
    bUseGravityOverride=true
    bUsePlanarLimit=false
    bUseSphericalLimits=true
    simulationSpace=RootRelative
    alphaBoolBlend {blendOption=Linear}
    boundBone {boneName=dyn_skirt_r_04}
    componentAppliedLinearAccClamp=<3000 3000 3000>
    componentLinearAccScale=<0.4 0.5 0.5>
    componentLinearVelScale=<0.4 0.5 0.5>
    physicsBodyDefinitions [{
      collisionType=CustomSphere
      sphereCollisionRadius=4
      boundBone {boneName=dyn_skirt_r_04}
      boxExtents=<17 8 8>
      constraintSetup {angularLimitsMax=<2.5 10 0>, angularLimitsMin=<-2.5 -10 -20>, angularTarget=<0 0 0>}
      localJointOffset=<2 0 0>
     }]
    planarLimits [{drivingBone {boneName=spine_02}, planeTransform {location=<0 -20 0>, rotation=<rot 0 0 -90>}}]
   }
   n38 = Component To Local(ComponentPose=@n37)
   n39 = Anim Dynamics - Bone: dyn_skirt_fr_02!(
    ComponentPose=Local To Component(LocalPose=Use cached pose 'Copy'())
    bAlphaBoolEnabled:bool=True
    GravityOverride:Vector="0,0,-980"
    ExternalForce="0,0,0"
   ) -> (Pose) {
    alphaInputType=Bool
    angularBiasOverride=0.999
    angularDampingOverride=0.999
    angularSpringConstant=1500
    bOverrideAngularBias=true
    bOverrideAngularDamping=true
    bUseGravityOverride=true
    bUsePlanarLimit=false
    bUseSphericalLimits=true
    simulationSpace=RootRelative
    alphaBoolBlend {blendOption=Linear}
    boundBone {boneName=dyn_skirt_fr_02}
    componentAppliedLinearAccClamp=<3000 3000 3000>
    componentLinearAccScale=<0.4 0.5 0.5>
    componentLinearVelScale=<0.4 0.5 0.5>
    physicsBodyDefinitions [{
      collisionType=CustomSphere
      sphereCollisionRadius=2
      boundBone {boneName=dyn_skirt_fr_02}
      boxExtents=<7 4 4>
      constraintSetup {angularLimitsMax=<2.5 30 0>, angularLimitsMin=<-2.5 3 -10>, angularTarget=<0 0 0>}
      localJointOffset=<2 0 0>
     }]
    planarLimits [{drivingBone {boneName=spine_02}, planeTransform {location=<0 -20 0>, rotation=<rot 0 0 -90>}}]
    sphericalLimits [
     {limitRadius=7, drivingBone {boneName=pelvis}, sphereLocalOffset=<0 0 0>}
     {limitRadius=7, drivingBone {boneName=spine_01}, sphereLocalOffset=<5 0 0>}
     {limitRadius=7, drivingBone {boneName=spine_02}, sphereLocalOffset=<7 0 0>}
    ]
   }
   n40 = Anim Dynamics - Bone: dyn_skirt_fr_03!(ComponentPose=@n39, bAlphaBoolEnabled:bool=True, GravityOverride:Vector="0,0,-980", ExternalForce="0,0,0") -> (Pose) {
    alphaInputType=Bool
    angularBiasOverride=0.999
    angularDampingOverride=0.999
    angularSpringConstant=1500
    bOverrideAngularBias=true
    bOverrideAngularDamping=true
    bUseGravityOverride=true
    bUsePlanarLimit=false
    bUseSphericalLimits=true
    simulationSpace=RootRelative
    alphaBoolBlend {blendOption=Linear}
    boundBone {boneName=dyn_skirt_fr_03}
    componentAppliedLinearAccClamp=<3000 3000 3000>
    componentLinearAccScale=<0.4 0.5 0.5>
    componentLinearVelScale=<0.4 0.5 0.5>
    physicsBodyDefinitions [{
      collisionType=CustomSphere
      sphereCollisionRadius=2
      boundBone {boneName=dyn_skirt_fr_03}
      boxExtents=<10 5.5 5.5>
      constraintSetup {angularLimitsMax=<2.5 20 0>, angularLimitsMin=<-2.5 -20 -20>, angularTarget=<0 0 0>}
      localJointOffset=<2 0 0>
     }]
    planarLimits [{drivingBone {boneName=spine_02}, planeTransform {location=<0 -20 0>, rotation=<rot 0 0 -90>}}]
   }
   n41 = Anim Dynamics - Bone: dyn_skirt_fr_04!(ComponentPose=@n40, bAlphaBoolEnabled:bool=True, GravityOverride:Vector="0,0,-980", ExternalForce="0,0,0") -> (Pose) {
    alphaInputType=Bool
    angularBiasOverride=0.999
    angularDampingOverride=0.999
    angularSpringConstant=1500
    bOverrideAngularBias=true
    bOverrideAngularDamping=true
    bUseGravityOverride=true
    bUsePlanarLimit=false
    bUseSphericalLimits=true
    simulationSpace=RootRelative
    alphaBoolBlend {blendOption=Linear}
    boundBone {boneName=dyn_skirt_fr_04}
    componentAppliedLinearAccClamp=<3000 3000 3000>
    componentLinearAccScale=<0.4 0.5 0.5>
    componentLinearVelScale=<0.4 0.5 0.5>
    physicsBodyDefinitions [{
      collisionType=CustomSphere
      sphereCollisionRadius=4
      boundBone {boneName=dyn_skirt_fr_04}
      boxExtents=<17 8 8>
      constraintSetup {angularLimitsMax=<2.5 10 0>, angularLimitsMin=<-2.5 -10 -20>, angularTarget=<0 0 0>}
      localJointOffset=<2 0 0>
     }]
    planarLimits [{drivingBone {boneName=spine_02}, planeTransform {location=<0 -20 0>, rotation=<rot 0 0 -90>}}]
   }
   n42 = Anim Dynamics - Bone: dyn_skirt_bk_02!(ComponentPose=@n41, bAlphaBoolEnabled:bool=True, GravityOverride:Vector="0,0,-980", ExternalForce="0,0,0") -> (Pose) {
    alphaInputType=Bool
    angularBiasOverride=0.999
    angularDampingOverride=0.999
    angularSpringConstant=1500
    bOverrideAngularBias=true
    bOverrideAngularDamping=true
    bUseGravityOverride=true
    bUsePlanarLimit=false
    bUseSphericalLimits=true
    simulationSpace=RootRelative
    alphaBoolBlend {blendOption=Linear}
    boundBone {boneName=dyn_skirt_bk_02}
    componentAppliedLinearAccClamp=<3000 3000 3000>
    componentLinearAccScale=<0.4 0.5 0.5>
    componentLinearVelScale=<0.4 0.5 0.5>
    physicsBodyDefinitions [{
      collisionType=CustomSphere
      sphereCollisionRadius=2
      boundBone {boneName=dyn_skirt_bk_02}
      boxExtents=<7 4 4>
      constraintSetup {angularLimitsMax=<2.5 30 0>, angularLimitsMin=<-2.5 3 -10>, angularTarget=<0 0 0>}
      localJointOffset=<2 0 0>
     }]
    planarLimits [{drivingBone {boneName=spine_02}, planeTransform {location=<0 -20 0>, rotation=<rot 0 0 -90>}}]
    sphericalLimits [
     {limitRadius=7, drivingBone {boneName=pelvis}, sphereLocalOffset=<0 0 0>}
     {limitRadius=7, drivingBone {boneName=spine_01}, sphereLocalOffset=<5 0 0>}
     {limitRadius=7, drivingBone {boneName=spine_02}, sphereLocalOffset=<7 0 0>}
    ]
   }
   n43 = Anim Dynamics - Bone: dyn_skirt_bk_03!(ComponentPose=@n42, bAlphaBoolEnabled:bool=True, GravityOverride:Vector="0,0,-980", ExternalForce="0,0,0") -> (Pose) {
    alphaInputType=Bool
    angularBiasOverride=0.999
    angularDampingOverride=0.999
    angularSpringConstant=1500
    bOverrideAngularBias=true
    bOverrideAngularDamping=true
    bUseGravityOverride=true
    bUsePlanarLimit=false
    bUseSphericalLimits=true
    simulationSpace=RootRelative
    alphaBoolBlend {blendOption=Linear}
    boundBone {boneName=dyn_skirt_bk_03}
    componentAppliedLinearAccClamp=<3000 3000 3000>
    componentLinearAccScale=<0.4 0.5 0.5>
    componentLinearVelScale=<0.4 0.5 0.5>
    physicsBodyDefinitions [{
      collisionType=CustomSphere
      sphereCollisionRadius=2
      boundBone {boneName=dyn_skirt_bk_03}
      boxExtents=<10 5.5 5.5>
      constraintSetup {angularLimitsMax=<2.5 20 0>, angularLimitsMin=<-2.5 -20 -20>, angularTarget=<0 0 0>}
      localJointOffset=<2 0 0>
     }]
    planarLimits [{drivingBone {boneName=spine_02}, planeTransform {location=<0 -20 0>, rotation=<rot 0 0 -90>}}]
   }
   n44 = Anim Dynamics - Bone: dyn_skirt_bk_04!(ComponentPose=@n43, bAlphaBoolEnabled:bool=True, GravityOverride:Vector="0,0,-980", ExternalForce="0,0,0") -> (Pose) {
    alphaInputType=Bool
    angularBiasOverride=0.999
    angularDampingOverride=0.999
    angularSpringConstant=1500
    bOverrideAngularBias=true
    bOverrideAngularDamping=true
    bUseGravityOverride=true
    bUsePlanarLimit=false
    bUseSphericalLimits=true
    simulationSpace=RootRelative
    alphaBoolBlend {blendOption=Linear}
    boundBone {boneName=dyn_skirt_bk_04}
    componentAppliedLinearAccClamp=<3000 3000 3000>
    componentLinearAccScale=<0.4 0.5 0.5>
    componentLinearVelScale=<0.4 0.5 0.5>
    physicsBodyDefinitions [{
      collisionType=CustomSphere
      sphereCollisionRadius=4
      boundBone {boneName=dyn_skirt_bk_04}
      boxExtents=<17 8 8>
      constraintSetup {angularLimitsMax=<2.5 10 0>, angularLimitsMin=<-2.5 -10 -20>, angularTarget=<0 0 0>}
      localJointOffset=<2 0 0>
     }]
    planarLimits [{drivingBone {boneName=spine_02}, planeTransform {location=<0 -20 0>, rotation=<rot 0 0 -90>}}]
   }
   n45 = Anim Dynamics - Bone: dyn_skirt_l_fr_02(ComponentPose=@n44, bAlphaBoolEnabled=True, GravityOverride="0,0,-980", ExternalForce="0,0,0") {
    alphaInputType=Bool
    angularBiasOverride=0.999
    angularDampingOverride=0.999
    angularSpringConstant=1500
    bOverrideAngularBias=true
    bOverrideAngularDamping=true
    bUseGravityOverride=true
    bUsePlanarLimit=false
    bUseSphericalLimits=true
    simulationSpace=RootRelative
    alphaBoolBlend {blendOption=Linear}
    boundBone {boneName=dyn_skirt_l_fr_02}
    componentAppliedLinearAccClamp=<3000 3000 3000>
    componentLinearAccScale=<0.4 0.5 0.5>
    componentLinearVelScale=<0.4 0.5 0.5>
    physicsBodyDefinitions [{
      collisionType=CustomSphere
      sphereCollisionRadius=2
      boundBone {boneName=dyn_skirt_l_fr_02}
      boxExtents=<7 4 4>
      constraintSetup {angularLimitsMax=<2.5 30 0>, angularLimitsMin=<-2.5 3 -10>, angularTarget=<0 0 0>}
      localJointOffset=<2 0 0>
     }]
    planarLimits [{drivingBone {boneName=spine_02}, planeTransform {location=<0 -20 0>, rotation=<rot 0 0 -90>}}]
    sphericalLimits [
     {limitRadius=7, drivingBone {boneName=pelvis}, sphereLocalOffset=<0 0 0>}
     {limitRadius=7, drivingBone {boneName=spine_01}, sphereLocalOffset=<5 0 0>}
     {limitRadius=7, drivingBone {boneName=spine_02}, sphereLocalOffset=<7 0 0>}
    ]
   }
   n46 = Anim Dynamics - Bone: dyn_skirt_l_fr_03(ComponentPose=@n45, bAlphaBoolEnabled=True, GravityOverride="0,0,-980", ExternalForce="0,0,0") {
    alphaInputType=Bool
    angularBiasOverride=0.999
    angularDampingOverride=0.999
    angularSpringConstant=1500
    bOverrideAngularBias=true
    bOverrideAngularDamping=true
    bUseGravityOverride=true
    bUsePlanarLimit=false
    bUseSphericalLimits=true
    simulationSpace=RootRelative
    alphaBoolBlend {blendOption=Linear}
    boundBone {boneName=dyn_skirt_l_fr_03}
    componentAppliedLinearAccClamp=<3000 3000 3000>
    componentLinearAccScale=<0.4 0.5 0.5>
    componentLinearVelScale=<0.4 0.5 0.5>
    physicsBodyDefinitions [{
      collisionType=CustomSphere
      sphereCollisionRadius=2
      boundBone {boneName=dyn_skirt_l_fr_03}
      boxExtents=<10 5.5 5.5>
      constraintSetup {angularLimitsMax=<2.5 20 0>, angularLimitsMin=<-2.5 -20 -20>, angularTarget=<0 0 0>}
      localJointOffset=<2 0 0>
     }]
    planarLimits [{drivingBone {boneName=spine_02}, planeTransform {location=<0 -20 0>, rotation=<rot 0 0 -90>}}]
   }
   n47 = Anim Dynamics - Bone: dyn_skirt_l_fr_04(ComponentPose=@n46, bAlphaBoolEnabled=True, GravityOverride="0,0,-980", ExternalForce="0,0,0") {
    alphaInputType=Bool
    angularBiasOverride=0.999
    angularDampingOverride=0.999
    angularSpringConstant=1500
    bOverrideAngularBias=true
    bOverrideAngularDamping=true
    bUseGravityOverride=true
    bUsePlanarLimit=false
    bUseSphericalLimits=true
    simulationSpace=RootRelative
    alphaBoolBlend {blendOption=Linear}
    boundBone {boneName=dyn_skirt_l_fr_04}
    componentAppliedLinearAccClamp=<3000 3000 3000>
    componentLinearAccScale=<0.4 0.5 0.5>
    componentLinearVelScale=<0.4 0.5 0.5>
    physicsBodyDefinitions [{
      collisionType=CustomSphere
      sphereCollisionRadius=4
      boundBone {boneName=dyn_skirt_l_fr_04}
      boxExtents=<17 8 8>
      constraintSetup {angularLimitsMax=<2.5 10 0>, angularLimitsMin=<-2.5 -10 -20>, angularTarget=<0 0 0>}
      localJointOffset=<2 0 0>
     }]
    planarLimits [{drivingBone {boneName=spine_02}, planeTransform {location=<0 -20 0>, rotation=<rot 0 0 -90>}}]
   }
   n48 = Anim Dynamics - Bone: dyn_skirt_l_03(ComponentPose=@n47, bAlphaBoolEnabled=True, GravityOverride="0,0,-980", ExternalForce="0,0,0") {
    alphaInputType=Bool
    angularBiasOverride=0.999
    angularDampingOverride=0.999
    angularSpringConstant=1500
    bOverrideAngularBias=true
    bOverrideAngularDamping=true
    bUseGravityOverride=true
    bUsePlanarLimit=false
    bUseSphericalLimits=true
    simulationSpace=RootRelative
    alphaBoolBlend {blendOption=Linear}
    boundBone {boneName=dyn_skirt_l_03}
    componentAppliedLinearAccClamp=<3000 3000 3000>
    componentLinearAccScale=<0.4 0.5 0.5>
    componentLinearVelScale=<0.4 0.5 0.5>
    physicsBodyDefinitions [{
      collisionType=CustomSphere
      sphereCollisionRadius=2
      boundBone {boneName=dyn_skirt_l_03}
      boxExtents=<10 5.5 5.5>
      constraintSetup {angularLimitsMax=<2.5 20 0>, angularLimitsMin=<-2.5 -20 -20>, angularTarget=<0 0 0>}
      localJointOffset=<2 0 0>
     }]
    planarLimits [{drivingBone {boneName=spine_02}, planeTransform {location=<0 -20 0>, rotation=<rot 0 0 -90>}}]
   }
   n49 = Anim Dynamics - Bone: dyn_skirt_l_04(ComponentPose=@n48, bAlphaBoolEnabled=True, GravityOverride="0,0,-980", ExternalForce="0,0,0") {
    alphaInputType=Bool
    angularBiasOverride=0.999
    angularDampingOverride=0.999
    angularSpringConstant=1500
    bOverrideAngularBias=true
    bOverrideAngularDamping=true
    bUseGravityOverride=true
    bUsePlanarLimit=false
    bUseSphericalLimits=true
    simulationSpace=RootRelative
    alphaBoolBlend {blendOption=Linear}
    boundBone {boneName=dyn_skirt_l_04}
    componentAppliedLinearAccClamp=<3000 3000 3000>
    componentLinearAccScale=<0.4 0.5 0.5>
    componentLinearVelScale=<0.4 0.5 0.5>
    physicsBodyDefinitions [{
      collisionType=CustomSphere
      sphereCollisionRadius=4
      boundBone {boneName=dyn_skirt_l_04}
      boxExtents=<17 8 8>
      constraintSetup {angularLimitsMax=<2.5 10 0>, angularLimitsMin=<-2.5 -10 -20>, angularTarget=<0 0 0>}
      localJointOffset=<2 0 0>
     }]
    planarLimits [{drivingBone {boneName=spine_02}, planeTransform {location=<0 -20 0>, rotation=<rot 0 0 -90>}}]
   }
   n50 = Anim Dynamics - Bone: dyn_skirt_l_bk_02!(ComponentPose=@n49, bAlphaBoolEnabled:bool=True, GravityOverride:Vector="0,0,-980", ExternalForce="0,0,0") -> (Pose) {
    alphaInputType=Bool
    angularBiasOverride=0.999
    angularDampingOverride=0.999
    angularSpringConstant=1500
    bOverrideAngularBias=true
    bOverrideAngularDamping=true
    bUseGravityOverride=true
    bUsePlanarLimit=false
    bUseSphericalLimits=true
    simulationSpace=RootRelative
    alphaBoolBlend {blendOption=Linear}
    boundBone {boneName=dyn_skirt_l_bk_02}
    componentAppliedLinearAccClamp=<3000 3000 3000>
    componentLinearAccScale=<0.4 0.5 0.5>
    componentLinearVelScale=<0.4 0.5 0.5>
    physicsBodyDefinitions [{
      collisionType=CustomSphere
      sphereCollisionRadius=2
      boundBone {boneName=dyn_skirt_l_bk_02}
      boxExtents=<7 4 4>
      constraintSetup {angularLimitsMax=<2.5 30 0>, angularLimitsMin=<-2.5 3 -10>, angularTarget=<0 0 0>}
      localJointOffset=<2 0 0>
     }]
    planarLimits [{drivingBone {boneName=spine_02}, planeTransform {location=<0 -20 0>, rotation=<rot 0 0 -90>}}]
    sphericalLimits [
     {limitRadius=7, drivingBone {boneName=pelvis}, sphereLocalOffset=<0 0 0>}
     {limitRadius=7, drivingBone {boneName=spine_01}, sphereLocalOffset=<5 0 0>}
     {limitRadius=7, drivingBone {boneName=spine_02}, sphereLocalOffset=<7 0 0>}
    ]
   }
   n51 = Anim Dynamics - Bone: dyn_skirt_l_bk_03!(ComponentPose=@n50, bAlphaBoolEnabled:bool=True, GravityOverride:Vector="0,0,-980", ExternalForce="0,0,0") -> (Pose) {
    alphaInputType=Bool
    angularBiasOverride=0.999
    angularDampingOverride=0.999
    angularSpringConstant=1500
    bOverrideAngularBias=true
    bOverrideAngularDamping=true
    bUseGravityOverride=true
    bUsePlanarLimit=false
    bUseSphericalLimits=true
    simulationSpace=RootRelative
    alphaBoolBlend {blendOption=Linear}
    boundBone {boneName=dyn_skirt_l_bk_03}
    componentAppliedLinearAccClamp=<3000 3000 3000>
    componentLinearAccScale=<0.4 0.5 0.5>
    componentLinearVelScale=<0.4 0.5 0.5>
    physicsBodyDefinitions [{
      collisionType=CustomSphere
      sphereCollisionRadius=2
      boundBone {boneName=dyn_skirt_l_bk_03}
      boxExtents=<10 5.5 5.5>
      constraintSetup {angularLimitsMax=<2.5 20 0>, angularLimitsMin=<-2.5 -20 -20>, angularTarget=<0 0 0>}
      localJointOffset=<2 0 0>
     }]
    planarLimits [{drivingBone {boneName=spine_02}, planeTransform {location=<0 -20 0>, rotation=<rot 0 0 -90>}}]
   }
   n52 = Anim Dynamics - Bone: dyn_skirt_l_bk_04!(ComponentPose=@n51, bAlphaBoolEnabled:bool=True, GravityOverride:Vector="0,0,-980", ExternalForce="0,0,0") -> (Pose) {
    alphaInputType=Bool
    angularBiasOverride=0.999
    angularDampingOverride=0.999
    angularSpringConstant=1500
    bOverrideAngularBias=true
    bOverrideAngularDamping=true
    bUseGravityOverride=true
    bUsePlanarLimit=false
    bUseSphericalLimits=true
    simulationSpace=RootRelative
    alphaBoolBlend {blendOption=Linear}
    boundBone {boneName=dyn_skirt_l_bk_04}
    componentAppliedLinearAccClamp=<3000 3000 3000>
    componentLinearAccScale=<0.4 0.5 0.5>
    componentLinearVelScale=<0.4 0.5 0.5>
    physicsBodyDefinitions [{
      collisionType=CustomSphere
      sphereCollisionRadius=4
      boundBone {boneName=dyn_skirt_l_bk_04}
      boxExtents=<17 8 8>
      constraintSetup {angularLimitsMax=<2.5 10 0>, angularLimitsMin=<-2.5 -10 -20>, angularTarget=<0 0 0>}
      localJointOffset=<2 0 0>
     }]
    planarLimits [{drivingBone {boneName=spine_02}, planeTransform {location=<0 -20 0>, rotation=<rot 0 0 -90>}}]
   }
   n53 = Anim Dynamics - Bone: dyn_skirt_r_fr_03(ComponentPose=@n52, bAlphaBoolEnabled=True, GravityOverride="0,0,-980", ExternalForce="0,0,0") {
    alphaInputType=Bool
    angularBiasOverride=0.999
    angularDampingOverride=0.999
    angularSpringConstant=1500
    bOverrideAngularBias=true
    bOverrideAngularDamping=true
    bUseGravityOverride=true
    bUsePlanarLimit=false
    bUseSphericalLimits=true
    simulationSpace=RootRelative
    alphaBoolBlend {blendOption=Linear}
    boundBone {boneName=dyn_skirt_r_fr_03}
    componentAppliedLinearAccClamp=<3000 3000 3000>
    componentLinearAccScale=<0.4 0.5 0.5>
    componentLinearVelScale=<0.4 0.5 0.5>
    physicsBodyDefinitions [{
      collisionType=CustomSphere
      sphereCollisionRadius=2
      boundBone {boneName=dyn_skirt_r_fr_03}
      boxExtents=<7 4 4>
      constraintSetup {angularLimitsMax=<2.5 30 0>, angularLimitsMin=<-2.5 3 -10>, angularTarget=<0 0 0>}
      localJointOffset=<2 0 0>
     }]
    planarLimits [{drivingBone {boneName=spine_02}, planeTransform {location=<0 -20 0>, rotation=<rot 0 0 -90>}}]
    sphericalLimits [
     {limitRadius=7, drivingBone {boneName=pelvis}, sphereLocalOffset=<0 0 0>}
     {limitRadius=7, drivingBone {boneName=spine_01}, sphereLocalOffset=<5 0 0>}
     {limitRadius=7, drivingBone {boneName=spine_02}, sphereLocalOffset=<7 0 0>}
    ]
   }
   n54 = Anim Dynamics - Bone: dyn_skirt_r_fr_04(ComponentPose=@n53, bAlphaBoolEnabled=True, GravityOverride="0,0,-980", ExternalForce="0,0,0") {
    alphaInputType=Bool
    angularBiasOverride=0.999
    angularDampingOverride=0.999
    angularSpringConstant=1500
    bOverrideAngularBias=true
    bOverrideAngularDamping=true
    bUseGravityOverride=true
    bUsePlanarLimit=false
    bUseSphericalLimits=true
    simulationSpace=RootRelative
    alphaBoolBlend {blendOption=Linear}
    boundBone {boneName=dyn_skirt_r_fr_04}
    componentAppliedLinearAccClamp=<3000 3000 3000>
    componentLinearAccScale=<0.4 0.5 0.5>
    componentLinearVelScale=<0.4 0.5 0.5>
    physicsBodyDefinitions [{
      collisionType=CustomSphere
      sphereCollisionRadius=2
      boundBone {boneName=dyn_skirt_r_fr_04}
      boxExtents=<10 5.5 5.5>
      constraintSetup {angularLimitsMax=<2.5 20 0>, angularLimitsMin=<-2.5 -20 -20>, angularTarget=<0 0 0>}
      localJointOffset=<2 0 0>
     }]
    planarLimits [{drivingBone {boneName=spine_02}, planeTransform {location=<0 -20 0>, rotation=<rot 0 0 -90>}}]
   }
   n55 = Anim Dynamics - Bone: dyn_skirt_r_fr_05(ComponentPose=@n54, bAlphaBoolEnabled=True, GravityOverride="0,0,-980", ExternalForce="0,0,0") {
    alphaInputType=Bool
    angularBiasOverride=0.999
    angularDampingOverride=0.999
    angularSpringConstant=1500
    bOverrideAngularBias=true
    bOverrideAngularDamping=true
    bUseGravityOverride=true
    bUsePlanarLimit=false
    bUseSphericalLimits=true
    simulationSpace=RootRelative
    alphaBoolBlend {blendOption=Linear}
    boundBone {boneName=dyn_skirt_r_fr_05}
    componentAppliedLinearAccClamp=<3000 3000 3000>
    componentLinearAccScale=<0.4 0.5 0.5>
    componentLinearVelScale=<0.4 0.5 0.5>
    physicsBodyDefinitions [{
      collisionType=CustomSphere
      sphereCollisionRadius=4
      boundBone {boneName=dyn_skirt_r_fr_05}
      boxExtents=<17 8 8>
      constraintSetup {angularLimitsMax=<2.5 10 0>, angularLimitsMin=<-2.5 -10 -20>, angularTarget=<0 0 0>}
      localJointOffset=<2 0 0>
     }]
    planarLimits [{drivingBone {boneName=spine_02}, planeTransform {location=<0 -20 0>, rotation=<rot 0 0 -90>}}]
   }
   n56 = Anim Dynamics - Bone: dyn_skirt_r_03(ComponentPose=@n55, bAlphaBoolEnabled=True, GravityOverride="0,0,-980", ExternalForce="0,0,0") {
    alphaInputType=Bool
    angularBiasOverride=0.999
    angularDampingOverride=0.999
    angularSpringConstant=1500
    bOverrideAngularBias=true
    bOverrideAngularDamping=true
    bUseGravityOverride=true
    bUsePlanarLimit=false
    bUseSphericalLimits=true
    simulationSpace=RootRelative
    alphaBoolBlend {blendOption=Linear}
    boundBone {boneName=dyn_skirt_r_03}
    componentAppliedLinearAccClamp=<3000 3000 3000>
    componentLinearAccScale=<0.4 0.5 0.5>
    componentLinearVelScale=<0.4 0.5 0.5>
    physicsBodyDefinitions [{
      collisionType=CustomSphere
      sphereCollisionRadius=2
      boundBone {boneName=dyn_skirt_r_03}
      boxExtents=<10 5.5 5.5>
      constraintSetup {angularLimitsMax=<2.5 20 0>, angularLimitsMin=<-2.5 -20 -20>, angularTarget=<0 0 0>}
      localJointOffset=<2 0 0>
     }]
    planarLimits [{drivingBone {boneName=spine_02}, planeTransform {location=<0 -20 0>, rotation=<rot 0 0 -90>}}]
   }
   n57 = Anim Dynamics - Bone: dyn_skirt_r_04(ComponentPose=@n56, bAlphaBoolEnabled=True, GravityOverride="0,0,-980", ExternalForce="0,0,0") {
    alphaInputType=Bool
    angularBiasOverride=0.999
    angularDampingOverride=0.999
    angularSpringConstant=1500
    bOverrideAngularBias=true
    bOverrideAngularDamping=true
    bUseGravityOverride=true
    bUsePlanarLimit=false
    bUseSphericalLimits=true
    simulationSpace=RootRelative
    alphaBoolBlend {blendOption=Linear}
    boundBone {boneName=dyn_skirt_r_04}
    componentAppliedLinearAccClamp=<3000 3000 3000>
    componentLinearAccScale=<0.4 0.5 0.5>
    componentLinearVelScale=<0.4 0.5 0.5>
    physicsBodyDefinitions [{
      collisionType=CustomSphere
      sphereCollisionRadius=4
      boundBone {boneName=dyn_skirt_r_04}
      boxExtents=<17 8 8>
      constraintSetup {angularLimitsMax=<2.5 10 0>, angularLimitsMin=<-2.5 -10 -20>, angularTarget=<0 0 0>}
      localJointOffset=<2 0 0>
     }]
    planarLimits [{drivingBone {boneName=spine_02}, planeTransform {location=<0 -20 0>, rotation=<rot 0 0 -90>}}]
   }
   n58 = Anim Dynamics - Bone: dyn_skirt_r_bk_02!(ComponentPose=@n57, bAlphaBoolEnabled:bool=True, GravityOverride:Vector="0,0,-980", ExternalForce="0,0,0") -> (Pose) {
    alphaInputType=Bool
    angularBiasOverride=0.999
    angularDampingOverride=0.999
    angularSpringConstant=1500
    bOverrideAngularBias=true
    bOverrideAngularDamping=true
    bUseGravityOverride=true
    bUsePlanarLimit=false
    bUseSphericalLimits=true
    simulationSpace=RootRelative
    alphaBoolBlend {blendOption=Linear}
    boundBone {boneName=dyn_skirt_r_bk_02}
    componentAppliedLinearAccClamp=<3000 3000 3000>
    componentLinearAccScale=<0.4 0.5 0.5>
    componentLinearVelScale=<0.4 0.5 0.5>
    physicsBodyDefinitions [{
      collisionType=CustomSphere
      sphereCollisionRadius=2
      boundBone {boneName=dyn_skirt_r_bk_02}
      boxExtents=<7 4 4>
      constraintSetup {angularLimitsMax=<2.5 30 0>, angularLimitsMin=<-2.5 3 -10>, angularTarget=<0 0 0>}
      localJointOffset=<2 0 0>
     }]
    planarLimits [{drivingBone {boneName=spine_02}, planeTransform {location=<0 -20 0>, rotation=<rot 0 0 -90>}}]
    sphericalLimits [
     {limitRadius=7, drivingBone {boneName=pelvis}, sphereLocalOffset=<0 0 0>}
     {limitRadius=7, drivingBone {boneName=spine_01}, sphereLocalOffset=<5 0 0>}
     {limitRadius=7, drivingBone {boneName=spine_02}, sphereLocalOffset=<7 0 0>}
    ]
   }
   n59 = Anim Dynamics - Bone: dyn_skirt_r_bk_03!(ComponentPose=@n58, bAlphaBoolEnabled:bool=True, GravityOverride:Vector="0,0,-980", ExternalForce="0,0,0") -> (Pose) {
    alphaInputType=Bool
    angularBiasOverride=0.999
    angularDampingOverride=0.999
    angularSpringConstant=1500
    bOverrideAngularBias=true
    bOverrideAngularDamping=true
    bUseGravityOverride=true
    bUsePlanarLimit=false
    bUseSphericalLimits=true
    simulationSpace=RootRelative
    alphaBoolBlend {blendOption=Linear}
    boundBone {boneName=dyn_skirt_r_bk_03}
    componentAppliedLinearAccClamp=<3000 3000 3000>
    componentLinearAccScale=<0.4 0.5 0.5>
    componentLinearVelScale=<0.4 0.5 0.5>
    physicsBodyDefinitions [{
      collisionType=CustomSphere
      sphereCollisionRadius=2
      boundBone {boneName=dyn_skirt_r_bk_03}
      boxExtents=<10 5.5 5.5>
      constraintSetup {angularLimitsMax=<2.5 20 0>, angularLimitsMin=<-2.5 -20 -20>, angularTarget=<0 0 0>}
      localJointOffset=<2 0 0>
     }]
    planarLimits [{drivingBone {boneName=spine_02}, planeTransform {location=<0 -20 0>, rotation=<rot 0 0 -90>}}]
   }
   n60 = Anim Dynamics - Bone: dyn_skirt_r_bk_04!(ComponentPose=@n59, bAlphaBoolEnabled:bool=True, GravityOverride:Vector="0,0,-980", ExternalForce="0,0,0") -> (Pose) {
    alphaInputType=Bool
    angularBiasOverride=0.999
    angularDampingOverride=0.999
    angularSpringConstant=1500
    bOverrideAngularBias=true
    bOverrideAngularDamping=true
    bUseGravityOverride=true
    bUsePlanarLimit=false
    bUseSphericalLimits=true
    simulationSpace=RootRelative
    alphaBoolBlend {blendOption=Linear}
    boundBone {boneName=dyn_skirt_r_bk_04}
    componentAppliedLinearAccClamp=<3000 3000 3000>
    componentLinearAccScale=<0.4 0.5 0.5>
    componentLinearVelScale=<0.4 0.5 0.5>
    physicsBodyDefinitions [{
      collisionType=CustomSphere
      sphereCollisionRadius=4
      boundBone {boneName=dyn_skirt_r_bk_04}
      boxExtents=<17 8 8>
      constraintSetup {angularLimitsMax=<2.5 10 0>, angularLimitsMin=<-2.5 -10 -20>, angularTarget=<0 0 0>}
      localJointOffset=<2 0 0>
     }]
    planarLimits [{drivingBone {boneName=spine_02}, planeTransform {location=<0 -20 0>, rotation=<rot 0 0 -90>}}]
   }
   n61 = "Transform (Modify) Bone - Bone: dyn_skirt_r_fr_MSTR"!(ComponentPose=@n60, Alpha=1, Rotation="-5,0,0") -> (Pose) {
    rotationMode=BMM_Additive
    rotationSpace=BCS_BoneSpace
    translationMode=BMM_Additive
    alphaBoolBlend {blendOption=Linear}
    boneToModify {boneName=dyn_skirt_r_fr_MSTR}
    rotation {pitch=-5}
   }
   n62 = "Transform (Modify) Bone - Bone: dyn_skirt_r_MSTR"!(ComponentPose=@n61, Alpha=1, Rotation="-4,0,0") -> (Pose) {
    rotationMode=BMM_Additive
    rotationSpace=BCS_BoneSpace
    translationMode=BMM_Additive
    alphaBoolBlend {blendOption=Linear}
    boneToModify {boneName=dyn_skirt_r_MSTR}
    rotation {pitch=-4}
   }
   n63 = "Transform (Modify) Bone - Bone: dyn_skirt_r_bk_MSTR"!(ComponentPose=@n62, Alpha=1, Rotation="-5,0,0") -> (Pose) {
    rotationMode=BMM_Additive
    rotationSpace=BCS_BoneSpace
    translationMode=BMM_Additive
    alphaBoolBlend {blendOption=Linear}
    boneToModify {boneName=dyn_skirt_r_bk_MSTR}
    rotation {pitch=-5}
   }
   n64 = "Transform (Modify) Bone - Bone: dyn_skirt_l_fr_MSTR"!(ComponentPose=@n63, Alpha=1, Rotation="-5,0,0") -> (Pose) {
    rotationMode=BMM_Additive
    rotationSpace=BCS_BoneSpace
    translationMode=BMM_Additive
    alphaBoolBlend {blendOption=Linear}
    boneToModify {boneName=dyn_skirt_l_fr_MSTR}
    rotation {pitch=-5}
   }
   n65 = "Transform (Modify) Bone - Bone: dyn_skirt_l_MSTR"!(ComponentPose=@n64, Alpha=1, Rotation="-4,0,0") -> (Pose) {
    rotationMode=BMM_Additive
    rotationSpace=BCS_BoneSpace
    translationMode=BMM_Additive
    alphaBoolBlend {blendOption=Linear}
    boneToModify {boneName=dyn_skirt_l_MSTR}
    rotation {pitch=-4}
   }
   n66 = "Transform (Modify) Bone - Bone: dyn_skirt_l_bk_MSTR"!(ComponentPose=@n65, Alpha=1, Rotation="-5,0,0") -> (Pose) {
    rotationMode=BMM_Additive
    rotationSpace=BCS_BoneSpace
    translationMode=BMM_Additive
    alphaBoolBlend {blendOption=Linear}
    boneToModify {boneName=dyn_skirt_l_bk_MSTR}
    rotation {pitch=-5}
   }
   n67 = Anim Dynamics - Bone: dyn_tie_fr_03(ComponentPose=@n66, Alpha=1, GravityOverride="0,0,-980", ExternalForce="0,0,0") {
    angularBiasOverride=0.99
    angularDampingOverride=0.9999
    angularSpringConstant=500
    bAngularSpring=true
    bLinearSpring=true
    bOverrideAngularDamping=true
    bUseGravityOverride=true
    bUsePlanarLimit=false
    bUseSphericalLimits=true
    linearSpringConstant=50
    simulationSpace=RootRelative
    alphaBoolBlend {blendOption=Linear}
    boundBone {boneName=dyn_tie_fr_03}
    componentAppliedLinearAccClamp=<3000 3000 3000>
    componentLinearAccScale=<0.5 1 1>
    componentLinearVelScale=<0.5 1 1>
    physicsBodyDefinitions [{
      collisionType=CustomSphere
      sphereCollisionRadius=2
      boundBone {boneName=dyn_tie_fr_03}
      boxExtents=<4 7 4>
      constraintSetup {angularTargetAxis=AxisY, twistAxis=AxisY, angularLimitsMax=<2.5 30 20>, angularLimitsMin=<-2.5 -45 -45>, angularTarget=<0 0 0>}
      localJointOffset=<2 0 0>
     }]
    planarLimits [{drivingBone {boneName=spine_02}, planeTransform {location=<0 -20 0>, rotation=<rot 0 0 -90>}}]
    sphericalLimits [
     {limitRadius=7, drivingBone {boneName=pelvis}, sphereLocalOffset=<0 0 0>}
     {limitRadius=7, drivingBone {boneName=spine_01}, sphereLocalOffset=<5 0 0>}
     {limitRadius=7, drivingBone {boneName=spine_02}, sphereLocalOffset=<7 0 0>}
    ]
   }
   n68 = Anim Dynamics - Bone: dyn_tie_fr_04(ComponentPose=@n67, Alpha=1, GravityOverride="0,0,-980", ExternalForce="0,0,0") {
    angularBiasOverride=0.99
    angularDampingOverride=0.9999
    angularSpringConstant=500
    bAngularSpring=true
    bLinearSpring=true
    bOverrideAngularDamping=true
    bUseGravityOverride=true
    bUsePlanarLimit=false
    bUseSphericalLimits=true
    linearSpringConstant=50
    simulationSpace=RootRelative
    alphaBoolBlend {blendOption=Linear}
    boundBone {boneName=dyn_tie_fr_04}
    componentAppliedLinearAccClamp=<3000 3000 3000>
    componentLinearAccScale=<0.5 1 1>
    componentLinearVelScale=<0.5 1 1>
    physicsBodyDefinitions [{
      collisionType=CustomSphere
      sphereCollisionRadius=2
      boundBone {boneName=dyn_tie_fr_04}
      boxExtents=<5.5 10 5.5>
      constraintSetup {angularTargetAxis=AxisY, twistAxis=AxisY, angularLimitsMax=<2.5 30 20>, angularLimitsMin=<-2.5 -45 -45>, angularTarget=<0 0 0>}
      localJointOffset=<2 0 0>
     }]
    planarLimits [{drivingBone {boneName=spine_02}, planeTransform {location=<0 -20 0>, rotation=<rot 0 0 -90>}}]
   }
   n69 = Anim Dynamics - Bone: dyn_tie_fr_05(ComponentPose=@n68, Alpha=1, GravityOverride="0,0,-980", ExternalForce="0,0,0") {
    angularBiasOverride=0.99
    angularDampingOverride=0.9999
    angularSpringConstant=500
    bAngularSpring=true
    bLinearSpring=true
    bOverrideAngularDamping=true
    bUseGravityOverride=true
    bUsePlanarLimit=false
    bUseSphericalLimits=true
    linearSpringConstant=50
    simulationSpace=RootRelative
    alphaBoolBlend {blendOption=Linear}
    boundBone {boneName=dyn_tie_fr_05}
    componentLinearAccScale=<1 1 1>
    componentLinearVelScale=<1 1 1>
    physicsBodyDefinitions [{
      collisionType=CustomSphere
      sphereCollisionRadius=4
      boundBone {boneName=dyn_tie_fr_05}
      boxExtents=<8 17 8>
      constraintSetup {angularTargetAxis=AxisY, twistAxis=AxisY, angularLimitsMax=<2.5 30 20>, angularLimitsMin=<-2.5 -45 -45>, angularTarget=<0 0 0>}
      localJointOffset=<2 0 0>
     }]
    planarLimits [{drivingBone {boneName=spine_02}, planeTransform {location=<0 -20 0>, rotation=<rot 0 0 -90>}}]
   }
   n70 = Anim Dynamics - Bone: dyn_tie_fr_06(ComponentPose=@n69, Alpha=1, GravityOverride="0,0,-980", ExternalForce="0,0,0") {
    angularBiasOverride=0.99
    angularDampingOverride=0.9999
    angularSpringConstant=500
    bAngularSpring=true
    bLinearSpring=true
    bOverrideAngularDamping=true
    bUseGravityOverride=true
    bUsePlanarLimit=false
    bUseSphericalLimits=true
    linearSpringConstant=50
    simulationSpace=RootRelative
    alphaBoolBlend {blendOption=Linear}
    boundBone {boneName=dyn_tie_fr_06}
    componentLinearAccScale=<1 1 1>
    componentLinearVelScale=<1 1 1>
    physicsBodyDefinitions [{
      collisionType=CustomSphere
      sphereCollisionRadius=4
      boundBone {boneName=dyn_tie_fr_06}
      boxExtents=<8 17 8>
      constraintSetup {angularTargetAxis=AxisY, twistAxis=AxisY, angularLimitsMax=<2.5 30 20>, angularLimitsMin=<-2.5 -45 -45>, angularTarget=<0 0 0>}
      localJointOffset=<2 0 0>
     }]
    planarLimits [{drivingBone {boneName=spine_02}, planeTransform {location=<0 -20 0>, rotation=<rot 0 0 -90>}}]
   }
   n71 = Anim Dynamics - Bone: dyn_tie_bk_02(ComponentPose=@n70, Alpha=1, GravityOverride="0,0,-980", ExternalForce="0,0,0") {
    angularBiasOverride=0.99
    angularDampingOverride=0.9999
    angularSpringConstant=500
    bAngularSpring=true
    bLinearSpring=true
    bOverrideAngularDamping=true
    bUseGravityOverride=true
    bUsePlanarLimit=false
    bUseSphericalLimits=true
    linearSpringConstant=50
    simulationSpace=RootRelative
    alphaBoolBlend {blendOption=Linear}
    boundBone {boneName=dyn_tie_bk_02}
    componentAppliedLinearAccClamp=<3000 3000 3000>
    componentLinearAccScale=<0.5 1 1>
    componentLinearVelScale=<0.5 1 1>
    physicsBodyDefinitions [{
      collisionType=CustomSphere
      sphereCollisionRadius=2
      boundBone {boneName=dyn_tie_bk_02}
      boxExtents=<4 7 4>
      constraintSetup {angularTargetAxis=AxisY, twistAxis=AxisY, angularLimitsMax=<2.5 30 20>, angularLimitsMin=<-2.5 -45 -45>, angularTarget=<0 0 0>}
      localJointOffset=<2 0 0>
     }]
    planarLimits [{drivingBone {boneName=spine_02}, planeTransform {location=<0 -20 0>, rotation=<rot 0 0 -90>}}]
    sphericalLimits [
     {limitRadius=7, drivingBone {boneName=pelvis}, sphereLocalOffset=<0 0 0>}
     {limitRadius=7, drivingBone {boneName=spine_01}, sphereLocalOffset=<5 0 0>}
     {limitRadius=7, drivingBone {boneName=spine_02}, sphereLocalOffset=<7 0 0>}
    ]
   }
   n72 = Anim Dynamics - Bone: dyn_tie_bk_03(ComponentPose=@n71, Alpha=1, GravityOverride="0,0,-980", ExternalForce="0,0,0") {
    angularBiasOverride=0.99
    angularDampingOverride=0.9999
    angularSpringConstant=500
    bAngularSpring=true
    bLinearSpring=true
    bOverrideAngularDamping=true
    bUseGravityOverride=true
    bUsePlanarLimit=false
    bUseSphericalLimits=true
    linearSpringConstant=50
    simulationSpace=RootRelative
    alphaBoolBlend {blendOption=Linear}
    boundBone {boneName=dyn_tie_bk_03}
    componentAppliedLinearAccClamp=<3000 3000 3000>
    componentLinearAccScale=<0.5 1 1>
    componentLinearVelScale=<0.5 1 1>
    physicsBodyDefinitions [{
      collisionType=CustomSphere
      sphereCollisionRadius=2
      boundBone {boneName=dyn_tie_bk_03}
      boxExtents=<5.5 10 5.5>
      constraintSetup {angularTargetAxis=AxisY, twistAxis=AxisY, angularLimitsMax=<2.5 30 20>, angularLimitsMin=<-2.5 -45 -45>, angularTarget=<0 0 0>}
      localJointOffset=<2 0 0>
     }]
    planarLimits [{drivingBone {boneName=spine_02}, planeTransform {location=<0 -20 0>, rotation=<rot 0 0 -90>}}]
   }
   n73 = Anim Dynamics - Bone: dyn_tie_bk_04(ComponentPose=@n72, Alpha=1, GravityOverride="0,0,-980", ExternalForce="0,0,0") {
    angularBiasOverride=0.99
    angularDampingOverride=0.9999
    angularSpringConstant=500
    bAngularSpring=true
    bLinearSpring=true
    bOverrideAngularDamping=true
    bUseGravityOverride=true
    bUsePlanarLimit=false
    bUseSphericalLimits=true
    linearSpringConstant=50
    simulationSpace=RootRelative
    alphaBoolBlend {blendOption=Linear}
    boundBone {boneName=dyn_tie_bk_04}
    componentLinearAccScale=<1 1 1>
    componentLinearVelScale=<1 1 1>
    physicsBodyDefinitions [{
      collisionType=CustomSphere
      sphereCollisionRadius=4
      boundBone {boneName=dyn_tie_bk_04}
      boxExtents=<8 17 8>
      constraintSetup {angularTargetAxis=AxisY, twistAxis=AxisY, angularLimitsMax=<2.5 30 20>, angularLimitsMin=<-2.5 -45 -45>, angularTarget=<0 0 0>}
      localJointOffset=<2 0 0>
     }]
    planarLimits [{drivingBone {boneName=spine_02}, planeTransform {location=<0 -20 0>, rotation=<rot 0 0 -90>}}]
   }
   n74 = Anim Dynamics - Bone: dyn_tie_bk_05(ComponentPose=@n73, Alpha=1, GravityOverride="0,0,-980", ExternalForce="0,0,0") {
    angularBiasOverride=0.99
    angularDampingOverride=0.9999
    angularSpringConstant=500
    bAngularSpring=true
    bLinearSpring=true
    bOverrideAngularDamping=true
    bUseGravityOverride=true
    bUsePlanarLimit=false
    bUseSphericalLimits=true
    linearSpringConstant=50
    simulationSpace=RootRelative
    alphaBoolBlend {blendOption=Linear}
    boundBone {boneName=dyn_tie_bk_05}
    componentLinearAccScale=<1 1 1>
    componentLinearVelScale=<1 1 1>
    physicsBodyDefinitions [{
      collisionType=CustomSphere
      sphereCollisionRadius=4
      boundBone {boneName=dyn_tie_bk_05}
      boxExtents=<8 17 8>
      constraintSetup {angularTargetAxis=AxisY, twistAxis=AxisY, angularLimitsMax=<2.5 30 20>, angularLimitsMin=<-2.5 -45 -45>, angularTarget=<0 0 0>}
      localJointOffset=<2 0 0>
     }]
    planarLimits [{drivingBone {boneName=spine_02}, planeTransform {location=<0 -20 0>, rotation=<rot 0 0 -90>}}]
   }
   n75 = "Constraint - dyn_tie_fr_01 (PPPPPPPPP)"!(
    ComponentPose=@n74
    Alpha=1
    ConstraintWeights_0=0.0959
    ConstraintWeights_1:float=0.1203
    ConstraintWeights_2:float=0.1481
    ConstraintWeights_3:float=0.007
    ConstraintWeights_4:float=0.3035
    ConstraintWeights_5:float=0.1502
    ConstraintWeights_6:float=0.1542
    ConstraintWeights_7:float=0.0065
    ConstraintWeights_8:float=0.0143
   ) -> (Pose) {
    alphaBoolBlend {blendOption=Linear}
    boneToModify {boneName=dyn_tie_fr_01}
    constraintSetup [
     {targetBone {boneName=pelvis}}
     {targetBone {boneName=dyn_skirt_fr_02}}
     {targetBone {boneName=dyn_skirt_fr_01}}
     {targetBone {boneName=dyn_skirt_l_fr_01}}
     {targetBone {boneName=dyn_skirt_r_fr_02}}
     {targetBone {boneName=dyn_skirt_r_fr_03}}
     {targetBone {boneName=dyn_skirt_r_fr_01}}
     {targetBone {boneName=dyn_skirt_r_02}}
     {targetBone {boneName=dyn_skirt_r_01}}
    ]
    constraintWeights [1,1,1,1,1,1,1,1,1]
   }
   n76 = "Constraint - dyn_tie_fr_02 (PPPPPPPPP)"!(
    ComponentPose=@n75
    Alpha=1
    ConstraintWeights_0=0.0156
    ConstraintWeights_1:float=0.212
    ConstraintWeights_2:float=0.0942
    ConstraintWeights_3:float=0.0111
    ConstraintWeights_4:float=0.113
    ConstraintWeights_5:float=0.2284
    ConstraintWeights_6:float=0.1671
    ConstraintWeights_7:float=0.0838
    ConstraintWeights_8:float=0.0747
   ) -> (Pose) {
    alphaBoolBlend {blendOption=Linear}
    boneToModify {boneName=dyn_tie_fr_02}
    constraintSetup [
     {targetBone {boneName=pelvis}}
     {targetBone {boneName=dyn_skirt_fr_02}}
     {targetBone {boneName=dyn_skirt_fr_03}}
     {targetBone {boneName=dyn_skirt_fr_04}}
     {targetBone {boneName=dyn_skirt_fr_01}}
     {targetBone {boneName=dyn_skirt_r_fr_02}}
     {targetBone {boneName=dyn_skirt_r_fr_03}}
     {targetBone {boneName=dyn_skirt_r_fr_04}}
     {targetBone {boneName=dyn_skirt_r_fr_01}}
    ]
    constraintWeights [1,1,1,1,1,1,1,1,1]
   }
   n77 = "Constraint - dyn_tie_fr_03 (PPPPPPPPP)"!(
    ComponentPose=@n76
    Alpha=1
    ConstraintWeights_0=0.1367
    ConstraintWeights_1:float=0.1753
    ConstraintWeights_2:float=0.233
    ConstraintWeights_3:float=0.0032
    ConstraintWeights_4:float=0.0062
    ConstraintWeights_5:float=0.0337
    ConstraintWeights_6:float=0.1487
    ConstraintWeights_7:float=0.1969
    ConstraintWeights_8:float=0.0664
   ) -> (Pose) {
    alphaBoolBlend {blendOption=Linear}
    boneToModify {boneName=dyn_tie_fr_03}
    constraintSetup [
     {targetBone {boneName=dyn_skirt_fr_02}}
     {targetBone {boneName=dyn_skirt_fr_03}}
     {targetBone {boneName=dyn_skirt_fr_04}}
     {targetBone {boneName=dyn_skirt_fr_01}}
     {targetBone {boneName=dyn_skirt_l_fr_02}}
     {targetBone {boneName=dyn_skirt_r_fr_02}}
     {targetBone {boneName=dyn_skirt_r_fr_03}}
     {targetBone {boneName=dyn_skirt_r_fr_04}}
     {targetBone {boneName=dyn_skirt_r_fr_05}}
    ]
    constraintWeights [1,1,1,1,1,1,1,1,1]
   }
   n78 = "Constraint - dyn_tie_fr_04 (PPPPPPPPP)"!(
    ComponentPose=@n77
    Alpha=1
    ConstraintWeights_0=0.0264
    ConstraintWeights_1:float=0.122
    ConstraintWeights_2:float=0.628
    ConstraintWeights_3:float=0.0009
    ConstraintWeights_4:float=0.0004
    ConstraintWeights_5:float=0.0024
    ConstraintWeights_6:float=0.0133
    ConstraintWeights_7:float=0.0877
    ConstraintWeights_8:float=0.1189
   ) -> (Pose) {
    alphaBoolBlend {blendOption=Linear}
    boneToModify {boneName=dyn_tie_fr_04}
    constraintSetup [
     {targetBone {boneName=dyn_skirt_fr_02}}
     {targetBone {boneName=dyn_skirt_fr_03}}
     {targetBone {boneName=dyn_skirt_fr_04}}
     {targetBone {boneName=dyn_skirt_l_fr_02}}
     {targetBone {boneName=dyn_skirt_l_fr_04}}
     {targetBone {boneName=dyn_skirt_r_fr_02}}
     {targetBone {boneName=dyn_skirt_r_fr_03}}
     {targetBone {boneName=dyn_skirt_r_fr_04}}
     {targetBone {boneName=dyn_skirt_r_fr_05}}
    ]
    constraintWeights [1,1,1,1,1,1,1,1,1]
   }
   n79 = "Constraint - dyn_tie_bk_01 (PPPPPPPPP)"!(
    ComponentPose=@n78
    Alpha=1
    ConstraintWeights_0=0.0959
    ConstraintWeights_1:float=0.1203
    ConstraintWeights_2:float=0.1481
    ConstraintWeights_3:float=0.007
    ConstraintWeights_4:float=0.3035
    ConstraintWeights_5:float=0.1502
    ConstraintWeights_6:float=0.1542
    ConstraintWeights_7:float=0.0065
    ConstraintWeights_8:float=0.0143
   ) -> (Pose) {
    alphaBoolBlend {blendOption=Linear}
    boneToModify {boneName=dyn_tie_bk_01}
    constraintSetup [
     {targetBone {boneName=pelvis}}
     {targetBone {boneName=dyn_skirt_fr_02}}
     {targetBone {boneName=dyn_skirt_fr_01}}
     {targetBone {boneName=dyn_skirt_l_fr_01}}
     {targetBone {boneName=dyn_skirt_r_fr_02}}
     {targetBone {boneName=dyn_skirt_r_fr_03}}
     {targetBone {boneName=dyn_skirt_r_fr_01}}
     {targetBone {boneName=dyn_skirt_r_02}}
     {targetBone {boneName=dyn_skirt_r_01}}
    ]
    constraintWeights [1,1,1,1,1,1,1,1,1]
   }
   n80 = "Constraint - dyn_tie_bk_02 (PPPPPPPPP)"!(
    ComponentPose=@n79
    Alpha=1
    ConstraintWeights_0=0.0156
    ConstraintWeights_1:float=0.212
    ConstraintWeights_2:float=0.0942
    ConstraintWeights_3:float=0.0111
    ConstraintWeights_4:float=0.113
    ConstraintWeights_5:float=0.2284
    ConstraintWeights_6:float=0.1671
    ConstraintWeights_7:float=0.0838
    ConstraintWeights_8:float=0.0747
   ) -> (Pose) {
    alphaBoolBlend {blendOption=Linear}
    boneToModify {boneName=dyn_tie_bk_02}
    constraintSetup [
     {targetBone {boneName=pelvis}}
     {targetBone {boneName=dyn_skirt_fr_02}}
     {targetBone {boneName=dyn_skirt_fr_03}}
     {targetBone {boneName=dyn_skirt_fr_04}}
     {targetBone {boneName=dyn_skirt_fr_01}}
     {targetBone {boneName=dyn_skirt_r_fr_02}}
     {targetBone {boneName=dyn_skirt_r_fr_03}}
     {targetBone {boneName=dyn_skirt_r_fr_04}}
     {targetBone {boneName=dyn_skirt_r_fr_01}}
    ]
    constraintWeights [1,1,1,1,1,1,1,1,1]
   }
   n81 = "Constraint - dyn_tie_bk_03 (PPPPPPPPP)"!(
    ComponentPose=@n80
    Alpha=1
    ConstraintWeights_0=0.1367
    ConstraintWeights_1:float=0.1753
    ConstraintWeights_2:float=0.233
    ConstraintWeights_3:float=0.0032
    ConstraintWeights_4:float=0.0062
    ConstraintWeights_5:float=0.0337
    ConstraintWeights_6:float=0.1487
    ConstraintWeights_7:float=0.1969
    ConstraintWeights_8:float=0.0664
   ) -> (Pose) {
    alphaBoolBlend {blendOption=Linear}
    boneToModify {boneName=dyn_tie_bk_03}
    constraintSetup [
     {targetBone {boneName=dyn_skirt_fr_02}}
     {targetBone {boneName=dyn_skirt_fr_03}}
     {targetBone {boneName=dyn_skirt_fr_04}}
     {targetBone {boneName=dyn_skirt_fr_01}}
     {targetBone {boneName=dyn_skirt_l_fr_02}}
     {targetBone {boneName=dyn_skirt_r_fr_02}}
     {targetBone {boneName=dyn_skirt_r_fr_03}}
     {targetBone {boneName=dyn_skirt_r_fr_04}}
     {targetBone {boneName=dyn_skirt_r_fr_05}}
    ]
    constraintWeights [1,1,1,1,1,1,1,1,1]
   }
   n82 = "Constraint - dyn_tie_bk_04 (PPPPPPPPP)"!(
    ComponentPose=@n81
    Alpha=1
    ConstraintWeights_0=0.0264
    ConstraintWeights_1:float=0.122
    ConstraintWeights_2:float=0.628
    ConstraintWeights_3:float=0.0009
    ConstraintWeights_4:float=0.0004
    ConstraintWeights_5:float=0.0024
    ConstraintWeights_6:float=0.0133
    ConstraintWeights_7:float=0.0877
    ConstraintWeights_8:float=0.1189
   ) -> (Pose) {
    alphaBoolBlend {blendOption=Linear}
    boneToModify {boneName=dyn_tie_bk_04}
    constraintSetup [
     {targetBone {boneName=dyn_skirt_fr_02}}
     {targetBone {boneName=dyn_skirt_fr_03}}
     {targetBone {boneName=dyn_skirt_fr_04}}
     {targetBone {boneName=dyn_skirt_l_fr_02}}
     {targetBone {boneName=dyn_skirt_l_fr_04}}
     {targetBone {boneName=dyn_skirt_r_fr_02}}
     {targetBone {boneName=dyn_skirt_r_fr_03}}
     {targetBone {boneName=dyn_skirt_r_fr_04}}
     {targetBone {boneName=dyn_skirt_r_fr_05}}
    ]
    constraintWeights [1,1,1,1,1,1,1,1,1]
   }
   n83 = Component To Local(ComponentPose=@n82)
   n84 = Get bRigidBodyEnabled()
   n85 = Anim Dynamics - Bone: dyn_pouch_fr(
    ComponentPose=Local To Component(LocalPose=Blend Poses by bool(BlendPose_0=@n38, BlendPose_1=@n83, bActiveValue=@n84) {blendType=Linear})
    Alpha=1
    ExternalForce="0,0,0"
   ) {
    angularDampingOverride=0.9
    angularSpringConstant=1500
    bGravityOverrideInSimSpace=true
    bOverrideAngularDamping=true
    gravityScale=0.15
    simulationSpace=BoneRelative
    alphaBoolBlend {blendOption=Linear}
    boundBone {boneName=dyn_pouch_fr}
    componentAppliedLinearAccClamp=<3000 3000 3000>
    physicsBodyDefinitions [{
      sphereCollisionRadius=0
      boundBone {boneName=dyn_pouch_fr}
      boxExtents=<10 20 20>
      constraintSetup {angularLimitsMax=<0 0 2>, angularLimitsMin=<0 0 -1>, angularTarget=<0 0 0>}
      localJointOffset=<8 0 0>
     }]
    relativeSpaceBone {boneName=pelvis}
   }
   n86 = Anim Dynamics - Bone: dyn_pouch_bk(ComponentPose=@n85, Alpha=1, ExternalForce="0,0,0") {
    angularDampingOverride=0.9
    angularSpringConstant=1500
    bGravityOverrideInSimSpace=true
    bOverrideAngularDamping=true
    gravityScale=0.15
    simulationSpace=BoneRelative
    alphaBoolBlend {blendOption=Linear}
    boundBone {boneName=dyn_pouch_bk}
    componentAppliedLinearAccClamp=<3000 3000 3000>
    physicsBodyDefinitions [{
      sphereCollisionRadius=0
      boundBone {boneName=dyn_pouch_bk}
      boxExtents=<10 20 20>
      constraintSetup {angularLimitsMax=<0 0 2>, angularLimitsMin=<0 0 -1>, angularTarget=<0 0 0>}
      localJointOffset=<8 0 0>
     }]
    relativeSpaceBone {boneName=pelvis}
   }
   n87 = Anim Dynamics - Bone: dyn_glasses(ComponentPose=@n86, Alpha=1, ExternalForce="0,0,0") {
    angularSpringConstant=1500
    bGravityOverrideInSimSpace=true
    gravityScale=0.1
    linearSpringConstant=1000
    simulationSpace=BoneRelative
    alphaBoolBlend {blendOption=Linear}
    boundBone {boneName=dyn_glasses}
    componentAppliedLinearAccClamp=<3000 3000 3000>
    componentLinearAccScale=<0.15 0.15 0.35>
    componentLinearVelScale=<0.15 0.15 0.35>
    gravityOverride {z=-980}
    physicsBodyDefinitions [{
      sphereCollisionRadius=0
      boundBone {boneName=dyn_glasses}
      boxExtents=<8 20 20>
      constraintSetup {angularLimitsMax=<1 2 5>, angularLimitsMin=<-1 -2 -5>, angularTarget=<0 0 0>}
      localJointOffset=<2 0 0>
     }]
    relativeSpaceBone {boneName=pelvis}
   }
   n88 = "Constraint - dyn_chain_fr_01 (P)"(ComponentPose=@n87, Alpha=0.5, ConstraintWeights_0=1) {
    alphaBoolBlend {blendOption=Linear}
    boneToModify {boneName=dyn_chain_fr_01}
    constraintSetup [{targetBone {boneName=dyn_skirt_fr_MSTR}}]
    constraintWeights [1]
   }
   n89 = Anim Dynamics - Bone: dyn_chain_mid_01!(ComponentPose=@n88, Alpha=0.5, GravityOverride:Vector="0,0,-980", ExternalForce="0,0,0") -> (Pose) {
    angularDampingOverride=0.99
    angularSpringConstant=200
    bAngularSpring=true
    bLinearSpring=true
    bOverrideAngularDamping=true
    bUseGravityOverride=true
    bUsePlanarLimit=false
    bUseSphericalLimits=true
    linearSpringConstant=50
    simulationSpace=BoneRelative
    alphaBoolBlend {blendOption=Linear}
    boundBone {boneName=dyn_chain_mid_01}
    componentAppliedLinearAccClamp=<3000 3000 3000>
    componentLinearAccScale=<0.5 0.3 0.3>
    componentLinearVelScale=<0.5 0.3 0.3>
    gravityOverride {z=-980}
    physicsBodyDefinitions [{
      boundBone {boneName=dyn_chain_mid_01}
      boxExtents=<20 10 10>
      constraintSetup {angularLimitsMax=<1 5 5>, angularLimitsMin=<-1 -20 -5>, angularTarget=<0 0 0>}
      localJointOffset=<4 0 0>
     }]
    relativeSpaceBone {boneName=root}
    sphericalLimits [{limitRadius=8, drivingBone {boneName=pelvis}, sphereLocalOffset=<-5 0 0>}]
   }
   n90 = Anim Dynamics - Bone: dyn_chain_mid_02!(ComponentPose=@n89, Alpha=0.5, GravityOverride:Vector="0,0,-980", ExternalForce="0,0,0") -> (Pose) {
    angularDampingOverride=0.99
    angularSpringConstant=200
    bAngularSpring=true
    bLinearSpring=true
    bOverrideAngularDamping=true
    bUseGravityOverride=true
    bUsePlanarLimit=false
    bUseSphericalLimits=true
    linearSpringConstant=50
    simulationSpace=BoneRelative
    alphaBoolBlend {blendOption=Linear}
    boundBone {boneName=dyn_chain_mid_02}
    componentAppliedLinearAccClamp=<3000 3000 3000>
    componentLinearAccScale=<0.5 0.3 0.3>
    componentLinearVelScale=<0.5 0.3 0.3>
    gravityOverride {z=-980}
    physicsBodyDefinitions [{
      boundBone {boneName=dyn_chain_mid_02}
      boxExtents=<20 10 10>
      constraintSetup {angularLimitsMax=<1 5 5>, angularLimitsMin=<-1 -20 -5>, angularTarget=<0 0 0>}
      localJointOffset=<4 0 0>
     }]
    relativeSpaceBone {boneName=root}
    sphericalLimits [{limitRadius=8, drivingBone {boneName=pelvis}, sphereLocalOffset=<-5 0 0>}]
   }
   n91 = Anim Dynamics - Bone: dyn_chain_bk_01!(ComponentPose=@n90, Alpha=0.5, GravityOverride:Vector="0,0,-980", ExternalForce="0,0,0") -> (Pose) {
    angularDampingOverride=0.99
    angularSpringConstant=200
    bAngularSpring=true
    bLinearSpring=true
    bOverrideAngularDamping=true
    bUseGravityOverride=true
    bUsePlanarLimit=false
    bUseSphericalLimits=true
    linearSpringConstant=50
    simulationSpace=BoneRelative
    alphaBoolBlend {blendOption=Linear}
    boundBone {boneName=dyn_chain_bk_01}
    componentAppliedLinearAccClamp=<3000 3000 3000>
    componentLinearAccScale=<0.5 0.3 0.3>
    componentLinearVelScale=<0.5 0.3 0.3>
    gravityOverride {z=-980}
    physicsBodyDefinitions [{
      boundBone {boneName=dyn_chain_bk_01}
      boxExtents=<20 10 10>
      constraintSetup {angularLimitsMax=<1 5 5>, angularLimitsMin=<-1 -10 -5>, angularTarget=<0 0 0>}
      localJointOffset=<4 0 0>
     }]
    relativeSpaceBone {boneName=root}
    sphericalLimits [{limitRadius=15, drivingBone {boneName=pelvis}, sphereLocalOffset=<-5 0 0>}]
   }
   n92 = Anim Dynamics - Bone: dyn_bow_top_01(ComponentPose=@n91, Alpha=1, ExternalForce="0,0,0") {
    angularDampingOverride=0.9999
    angularSpringConstant=500
    bAngularSpring=true
    bGravityOverrideInSimSpace=true
    bOverrideAngularDamping=true
    bUseSphericalLimits=true
    simulationSpace=BoneRelative
    alphaBoolBlend {blendOption=Linear}
    boundBone {boneName=dyn_bow_top_01}
    componentAppliedLinearAccClamp=<3000 3000 3000>
    componentLinearAccScale=<0.7 0.7 0.7>
    componentLinearVelScale=<0.7 0.7 0.7>
    physicsBodyDefinitions [{
      sphereCollisionRadius=0
      boundBone {boneName=dyn_bow_top_01}
      boxExtents=<5 20 20>
      constraintSetup {angularLimitsMax=<1 10 10>, angularLimitsMin=<-1 -10 -10>, angularTarget=<0 0 0>}
      localJointOffset=<5 0 0>
     }]
    relativeSpaceBone {boneName=pelvis}
    sphericalLimits [{limitRadius=5, drivingBone {boneName=dyn_knot_MSTR}, sphereLocalOffset=<0 0 0>}]
   }
   n93 = Anim Dynamics - Bone: dyn_bow_top_02(ComponentPose=@n92, Alpha=1, ExternalForce="0,0,0") {
    angularDampingOverride=0.9999
    angularSpringConstant=500
    bAngularSpring=true
    bGravityOverrideInSimSpace=true
    bOverrideAngularDamping=true
    bUseSphericalLimits=true
    simulationSpace=BoneRelative
    alphaBoolBlend {blendOption=Linear}
    boundBone {boneName=dyn_bow_top_02}
    componentAppliedLinearAccClamp=<3000 3000 3000>
    componentLinearAccScale=<0.7 0.7 0.7>
    componentLinearVelScale=<0.7 0.7 0.7>
    physicsBodyDefinitions [{
      sphereCollisionRadius=0
      boundBone {boneName=dyn_bow_top_02}
      boxExtents=<5 20 20>
      constraintSetup {angularLimitsMax=<1 10 10>, angularLimitsMin=<-1 -10 -10>, angularTarget=<0 0 0>}
      localJointOffset=<5 0 0>
     }]
    relativeSpaceBone {boneName=pelvis}
    sphericalLimits [{limitRadius=5, drivingBone {boneName=dyn_knot_MSTR}, sphereLocalOffset=<0 0 0>}]
   }
   n94 = Anim Dynamics - Bone: dyn_bow_bot_01(ComponentPose=@n93, Alpha=1, ExternalForce="0,0,0") {
    angularDampingOverride=0.9999
    angularSpringConstant=500
    bAngularSpring=true
    bGravityOverrideInSimSpace=true
    bOverrideAngularDamping=true
    bUseSphericalLimits=true
    simulationSpace=BoneRelative
    alphaBoolBlend {blendOption=Linear}
    boundBone {boneName=dyn_bow_bot_01}
    componentAppliedLinearAccClamp=<3000 3000 3000>
    componentLinearAccScale=<0.7 0.7 0.7>
    componentLinearVelScale=<0.7 0.7 0.7>
    physicsBodyDefinitions [{
      sphereCollisionRadius=0
      boundBone {boneName=dyn_bow_bot_01}
      boxExtents=<5 20 20>
      constraintSetup {angularLimitsMax=<1 10 10>, angularLimitsMin=<-1 -10 -10>, angularTarget=<0 0 0>}
      localJointOffset=<5 0 0>
     }]
    relativeSpaceBone {boneName=pelvis}
    sphericalLimits [{limitRadius=5, drivingBone {boneName=dyn_knot_MSTR}, sphereLocalOffset=<0 0 0>}]
   }
   n95 = Anim Dynamics - Bone: dyn_bow_bot_02(ComponentPose=@n94, Alpha=1, ExternalForce="0,0,0") {
    angularDampingOverride=0.9999
    angularSpringConstant=500
    bAngularSpring=true
    bGravityOverrideInSimSpace=true
    bOverrideAngularDamping=true
    bUseSphericalLimits=true
    simulationSpace=BoneRelative
    alphaBoolBlend {blendOption=Linear}
    boundBone {boneName=dyn_bow_bot_02}
    componentAppliedLinearAccClamp=<3000 3000 3000>
    componentLinearAccScale=<0.7 0.7 0.7>
    componentLinearVelScale=<0.7 0.7 0.7>
    physicsBodyDefinitions [{
      sphereCollisionRadius=0
      boundBone {boneName=dyn_bow_bot_02}
      boxExtents=<5 20 20>
      constraintSetup {angularLimitsMax=<1 10 10>, angularLimitsMin=<-1 -10 -10>, angularTarget=<0 0 0>}
      localJointOffset=<5 0 0>
     }]
    relativeSpaceBone {boneName=pelvis}
    sphericalLimits [{limitRadius=5, drivingBone {boneName=dyn_knot_MSTR}, sphereLocalOffset=<0 0 0>}]
   }
   Save cached pose 'Pose'(Pose=Component To Local(ComponentPose=@n95)) {cacheName=Pose}
  ]}
 EventGraph {nodes [
   Event BlueprintUpdateAnimation()
   > n1 = Sequence!(execute) -> (then_0, then_1, then_2:Exec, then_3:Exec, then_4:Exec)
   > n2 = Gravity Override State Function DEPRECATED(
    CharacterPartAnimInstance=Self-Reference()
    GameTimeInSeconds=Get Delta Time X() -> (Delta Time X:double)
    GravityOverrideParams=Get Gravity Override Params()
   )
   > Set Gravity Override(Gravity Override=@n2)
   n3 = PlayerMovement(execute=@n1.then_1, AnimInstance=Self-Reference(), XDotProductCoefficient=100, YDotProductCoefficient=100, ZDotProductCoefficient=100)
   n4 = BreakVector(InVec=MakeVector(X=@n3, Y=@n3.YMovement))
   n5 = "float < float"(A=@n4, B=-50)
   > Set Left(Left:double=ClampAxis!(Angle:float="To Float (Boolean)"(InBool=@n5)) -> (ReturnValue:float)) -> (Output_Get:double)
   n6 = "float > float"(A=@n4, B=50)
   > Set Right(Right:double=ClampAxis(Angle="To Float (Boolean)"(InBool=@n6))) -> (Output_Get:double)
   n7 = "float > float"(A=@n4.Y, B=40)
   > Set Forward(Forward:double=ClampAxis(Angle="To Float (Boolean)"(InBool=@n7))) -> (Output_Get:double)
   n8 = "float < float"(A=@n4.Y, B=-40)
   > Set Backward(Backward:double=ClampAxis(Angle="To Float (Boolean)"(InBool=@n8))) -> (Output_Get:double)
   n9 = Branch(execute=@n1.then_2, Condition=Get bIsSliding())
   > n10 = Set Left_Calf(Left_Calf:double=GetAlphaFromReferenceTransformDelta(SkeletalMeshComponent=GetPartSkeletalMeshComponent(), BoneName=calf_l, BaseName=thigh_l, InputRange=(X=0,Y=1), TwistAxis="1,0,0").RawOutput) -> (Output_Get:double)
   > n11 = Branch(Condition=NOT Boolean(A="float > float"(A=@n10, B=100)))
   > Set BaseBall_Slide(BaseBall_Slide:bool=true) -> (Output_Get:bool)
   > Set Knee_Slide(Knee_Slide:bool=false) -> (Output_Get:bool)
   Set Knee_Slide(execute=@n11.else, Knee_Slide=true)
   > Set BaseBall_Slide(BaseBall_Slide=false)
   Set BaseBall_Slide(execute=@n9.else, BaseBall_Slide=false)
   > Set Knee_Slide(Knee_Slide=false)
   Set melee(execute=@n1.then_3, melee:double=SelectFloat(A=1, B=0, bPickA=Get bIsPlayingMeleeAnim())) -> (Output_Get:double)
   > Set Jumps(Jumps:bool=OR Boolean(A=Get bIsJumping(), B=Get bIsFalling())) -> (Output_Get:bool)
   > Set Crouch_off(Crouch_off:double=SelectFloat(A=0, B=1, bPickA=Get bIsCrouching())) -> (Output_Get:double)
   n12 = Get bIsPlayingEmote()
   Set FE_Emote(execute=@n1.then_4, FE_Emote:bool=AND Boolean(A=Get bIsInFrontEnd(), B=NOT Boolean(A=@n12))) -> (Output_Get:bool)
  ]}
]
variables [
 Backward {type=number}
 BaseBall_Slide {type=boolean}
 Crouch_off {type=number}
 Delta Time X {type=number}
 FE_Emote {type=boolean}
 Forward {type=number}
 Gravity Override {type=Vector}
 Gravity Override Params {type=GravityOverrideParamsStruct}
 Jumps {type=boolean}
 Knee_Slide {type=boolean}
 Left {type=number}
 Left_Calf {type=number}
 melee {type=number}
 Moving {type=number}
 Right {type=number}
]
