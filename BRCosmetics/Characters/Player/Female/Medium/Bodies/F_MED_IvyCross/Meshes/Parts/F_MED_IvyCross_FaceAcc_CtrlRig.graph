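# Control Rig for Ivy Cross's FaceAcc part that corrects ponytail and hair-strand bone poses.
# Spherical pose readers on the head and right upper arm, relative to neck_01, weight additive rotation and translation offsets.
# Melee alpha constrains a ponytail bone to the left clavicle, Target alpha shifts ponytail bones, and Emote or Skydive branches apply fixed offsets.
# A final debug-drawing reader chain is disconnected from execution.
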ControlRigBlueprint /BRCosmetics/Characters/Player/Female/Medium/Bodies/F_MED_IvyCross/Meshes/Parts/F_MED_IvyCross_FaceAcc_CtrlRig
graphs [
 RigVMFunctionLibrary {links [], nodes []}
 RigVMModel {nodes [
   RigUnit_BeginExecution = &RigUnit_BeginExecution()
   n1 = &RigUnit_SphericalPoseReader(
    ActiveRegionScaleFactors=(PositiveWidth=0.7,NegativeWidth=0.2,PositiveHeight=1,NegativeHeight=1)
    ActiveRegionSize=0.2
    Debug=(bDrawDebug=false,bDraw2D=false,bDrawLocalAxes=false,DebugScale=25,DebugSegments=20,DebugThickness=0.25)
    DriverAxis=(X=1,Y=0,Z=0)
    DriverItem=(Type=Bone,Name="head")
    FalloffRegionScaleFactors=(PositiveWidth=0.25,NegativeWidth=0.1,PositiveHeight=0.5,NegativeHeight=0.5)
    FalloffSize=0.45
    OptionalParentItem=(Type=Bone,Name="neck_01")
    RotationOffset=(X=-90,Y=0,Z=50)
    ExecutePin=@RigUnit_BeginExecution.ExecutePin
   )
   > &RigUnit_ModifyTransforms(
    ItemToModify=((Item=(Type=Bone,Name="dyn_ponytail_05"),Transform=(Rotation=(X=0,Y=0,Z=0,W=1),Translation=(X=0,Y=0,Z=0),Scale3D=(X=1,Y=1,Z=1))))
    ItemToModify.0.Transform.Rotation=&RigVMFunction_MathQuaternionFromEuler(Euler=(X=0,Y=15,Z=0), RotationOrder=XYZ).Result
    Weight=@n1.OutputParam
   )
   > &RigUnit_ModifyTransforms(
    ItemToModify="((Item=(Type=Bone,Name=\"dyn_ponytail_06\"),Transform=(Rotation=(X=0,Y=0,Z=0,W=1),Translation=(X=0,Y=-3,Z=0),Scale3D=(X=1,Y=1,Z=1))),(Item=(Type=Bone,Name=\"dyn_ponytail_08\"),Transform=(Rotation=(X=0,Y=0,Z=0,W=1),Translation=(X=0,Y=0,Z=0),Scale3D=(X=1,Y=1,Z=1 ...(truncated)"
    ItemToModify.0.Transform.Rotation=&RigVMFunction_MathQuaternionFromEuler(Euler=(X=0,Y=10,Z=5), RotationOrder=XYZ).Result
    ItemToModify.1.Transform.Rotation=&RigVMFunction_MathQuaternionFromEuler(Euler=(X=0,Y=50,Z=0), RotationOrder=XYZ).Result
    Weight=@n1.OutputParam
   )
   > &RigUnit_ModifyTransforms(
    ItemToModify=((Item=(Type=Bone,Name="dyn_hair_strand_r_bk_01"),Transform=(Rotation=(X=0,Y=0,Z=0,W=1),Translation=(X=0,Y=0,Z=0),Scale3D=(X=1,Y=1,Z=1))))
    ItemToModify.0.Transform.Rotation=&RigVMFunction_MathQuaternionFromEuler(Euler=(X=0,Y=-20,Z=-2), RotationOrder=XYZ).Result
    Weight=@n1.OutputParam
   )
   > n2 = &RigUnit_SphericalPoseReader(
    ActiveRegionScaleFactors=(PositiveWidth=0.7,NegativeWidth=0.2,PositiveHeight=0.7,NegativeHeight=0.5)
    ActiveRegionSize=0.2
    Debug=(bDrawDebug=false,bDraw2D=false,bDrawLocalAxes=false,DebugScale=25,DebugSegments=20,DebugThickness=0.25)
    DriverAxis=(X=1,Y=0,Z=0)
    DriverItem=(Type=Bone,Name="head")
    FalloffRegionScaleFactors=(PositiveWidth=0.4,NegativeWidth=0.1,PositiveHeight=0.4,NegativeHeight=0.2)
    FalloffSize=0.45
    OptionalParentItem=(Type=Bone,Name="neck_01")
    RotationOffset=(X=-90,Y=0,Z=110)
   )
   > &RigUnit_ModifyTransforms(
    ItemToModify=((Item=(Type=Bone,Name="dyn_ponytail_09"),Transform=(Rotation=(X=0,Y=0,Z=0,W=1),Translation=(X=0,Y=0,Z=0),Scale3D=(X=1,Y=1,Z=1))))
    ItemToModify.0.Transform.Rotation=&RigVMFunction_MathQuaternionFromEuler(Euler=(X=0,Y=0,Z=10)).Result
    Weight=@n2.OutputParam
   )
   n3 = &RigVMFunction_AlphaInterp(
    InterpSpeedDecreasing=2
    InterpSpeedIncreasing=10
    Scale=1
    bClampResult=False
    bInterpResult=False
    Value=RigVMVariableNode() {variable=Melee}.Value
   )
   > &RigUnit_ParentConstraint(
    AdvancedSettings=(RotationOrderForFilter=XZY)
    Child=(Type=Bone,Name="dyn_ponytail_08")
    Parents=((Item=(Type=Bone,Name="clavicle_l"),Weight=1))
    Weight=@n3.Result
   )
   > n4 = &RigUnit_SphericalPoseReader(
    ActiveRegionScaleFactors=(PositiveWidth=0.5,NegativeWidth=0.1,PositiveHeight=1,NegativeHeight=1)
    ActiveRegionSize=0.2
    Debug=(bDrawDebug=false,bDraw2D=false,bDrawLocalAxes=false,DebugScale=25,DebugSegments=20,DebugThickness=0.25)
    DriverAxis=(X=1,Y=0,Z=0)
    DriverItem=(Type=Bone,Name="head")
    FalloffRegionScaleFactors=(PositiveWidth=0.3,NegativeWidth=0.1,PositiveHeight=0.5,NegativeHeight=0.5)
    FalloffSize=0.45
    OptionalParentItem=(Type=Bone,Name="neck_01")
    RotationOffset=(X=-90,Y=0,Z=60)
   )
   > &RigUnit_ModifyTransforms(
    ItemToModify="((Item=(Type=Bone,Name=\"dyn_hair_strand_r_bk_01\"),Transform=(Rotation=(X=0,Y=0,Z=0,W=1),Translation=(X=0,Y=0,Z=0),Scale3D=(X=1,Y=1,Z=1))),(Item=(Type=Bone,Name=\"dyn_hair_strand_r_bk_07\"),Transform=(Rotation=(X=0,Y=0,Z=0,W=1),Translation=(X=0,Y=0,Z=0),Scale ...(truncated)"
    ItemToModify.0.Transform.Rotation=&RigVMFunction_MathQuaternionFromEuler(Euler=(X=0,Y=20,Z=20)).Result
    Weight=@n4.OutputParam
   )
   > &RigUnit_ModifyTransforms(
    ItemToModify=((Item=(Type=Bone,Name="dyn_hair_strand_r_01"),Transform=(Rotation=(X=0,Y=0,Z=0,W=1),Translation=(X=0,Y=0,Z=0),Scale3D=(X=1,Y=1,Z=1))))
    ItemToModify.0.Transform.Rotation=&RigVMFunction_MathQuaternionFromEuler(Euler=(X=0,Y=-10,Z=10)).Result
    Weight=@n4.OutputParam
   )
   > &RigUnit_ModifyTransforms(
    ItemToModify=((Item=(Type=Bone,Name="dyn_hair_strand_l_fr_01"),Transform=(Rotation=(X=0,Y=0,Z=0,W=1),Translation=(X=0,Y=0,Z=0),Scale3D=(X=1,Y=1,Z=1))))
    ItemToModify.0.Transform.Rotation=&RigVMFunction_MathQuaternionFromEuler(Euler=(X=0,Y=0,Z=-5)).Result
    Weight=@n4.OutputParam
   )
   > &RigUnit_ModifyTransforms(
    ItemToModify="((Item=(Type=Bone,Name=\"dyn_hair_strand_l_01\"),Transform=(Rotation=(X=0,Y=0,Z=0,W=1),Translation=(X=0,Y=0,Z=0),Scale3D=(X=1,Y=1,Z=1))),(Item=(Type=Bone,Name=\"dyn_hair_strand_l_07\"),Transform=(Rotation=(X=0,Y=0,Z=0,W=1),Translation=(X=0,Y=0,Z=0),Scale3D=(X= ...(truncated)"
    ItemToModify.0.Transform.Rotation=&RigVMFunction_MathQuaternionFromEuler(Euler=(X=0,Y=0,Z=-8)).Result
    ItemToModify.1.Transform.Rotation=&RigVMFunction_MathQuaternionFromEuler(Euler=(X=0,Y=0,Z=-5)).Result
    Weight=@n4.OutputParam
   )
   > &RigUnit_ModifyTransforms(
    ItemToModify="((Item=(Type=Bone,Name=\"dyn_hair_strand_l_bk_02\"),Transform=(Rotation=(X=0,Y=0,Z=0,W=1),Translation=(X=0,Y=0,Z=0),Scale3D=(X=1,Y=1,Z=1))),(Item=(Type=Bone,Name=\"dyn_hair_strand_l_bk_05\"),Transform=(Rotation=(X=0,Y=0,Z=0,W=1),Translation=(X=0,Y=0,Z=0),Scale ...(truncated)"
    ItemToModify.0.Transform.Rotation=&RigVMFunction_MathQuaternionFromEuler(Euler=(X=0,Y=8,Z=-5)).Result
    ItemToModify.1.Transform.Rotation=&RigVMFunction_MathQuaternionFromEuler(Euler=(X=0,Y=-20,Z=20)).Result
    Weight=@n4.OutputParam
   )
   > n5 = &RigUnit_SphericalPoseReader(
    ActiveRegionScaleFactors=(PositiveWidth=0.7,NegativeWidth=0.2,PositiveHeight=0.7,NegativeHeight=0.5)
    ActiveRegionSize=0.2
    Debug=(bDrawDebug=false,DebugScale=25,DebugSegments=20,DebugThickness=0.25)
    DriverAxis=(X=1,Y=0,Z=0)
    DriverItem=(Type=Bone,Name="head")
    FalloffRegionScaleFactors=(PositiveWidth=0.4,NegativeWidth=0.1,PositiveHeight=0.4,NegativeHeight=0.2)
    FalloffSize=0.45
    OptionalParentItem=(Type=Bone,Name="neck_01")
    RotationOffset=(X=-90,Y=0,Z=110)
   )
   > &RigUnit_ModifyTransforms(
    ItemToModify="((Item=(Type=Bone,Name=\"dyn_hair_strand_r_bk_01\"),Transform=(Rotation=(X=0,Y=0,Z=0,W=1),Translation=(X=0,Y=0,Z=0),Scale3D=(X=1,Y=1,Z=1))),(Item=(Type=Bone,Name=\"dyn_hair_strand_r_bk_06\"),Transform=(Rotation=(X=0,Y=0,Z=0,W=1),Translation=(X=0,Y=0,Z=0),Scale ...(truncated)"
    ItemToModify.1.Transform.Rotation=&RigVMFunction_MathQuaternionFromEuler(Euler=(X=0,Y=0,Z=0)).Result
    ItemToModify.0.Transform.Rotation=&RigVMFunction_MathQuaternionFromEuler(Euler=(X=0,Y=0,Z=15)).Result
    Weight=@n5.OutputParam
   )
   > &RigUnit_ModifyTransforms(
    ItemToModify=((Item=(Type=Bone,Name="dyn_hair_strand_r_01"),Transform=(Rotation=(X=0,Y=0,Z=0,W=1),Translation=(X=0,Y=0,Z=0),Scale3D=(X=1,Y=1,Z=1))))
    ItemToModify.0.Transform.Rotation=&RigVMFunction_MathQuaternionFromEuler(Euler=(X=0,Y=0,Z=10)).Result
    Weight=@n5.OutputParam
   )
   n6 = &RigVMFunction_MathQuaternionFromEuler(Euler=(X=0,Y=0,Z=6))
   > &RigUnit_ModifyTransforms(
    ItemToModify=((Item=(Type=Bone,Name="dyn_hair_strand_l_fr_01"),Transform=(Rotation=(X=0,Y=0,Z=0,W=1),Translation=(X=0,Y=0,Z=0),Scale3D=(X=1,Y=1,Z=1))))
    ItemToModify.0.Transform.Rotation=@n6.Result
    Weight=@n5.OutputParam
   )
   > &RigUnit_ModifyTransforms(
    ItemToModify="((Item=(Type=Bone,Name=\"dyn_hair_strand_l_01\"),Transform=(Rotation=(X=0,Y=0,Z=0,W=1),Translation=(X=0,Y=0,Z=0),Scale3D=(X=1,Y=1,Z=1))),(Item=(Type=Bone,Name=\"dyn_hair_strand_l_05\"),Transform=(Rotation=(X=0,Y=0,Z=0,W=1),Translation=(X=0,Y=0,Z=0),Scale3D=(X= ...(truncated)"
    ItemToModify.0.Transform.Rotation=&RigVMFunction_MathQuaternionFromEuler(Euler=(X=0,Y=0,Z=6)).Result
    ItemToModify.1.Transform.Rotation=@n6.Result
    Weight=@n5.OutputParam
   )
   > &RigUnit_ModifyTransforms(
    ItemToModify="((Item=(Type=Bone,Name=\"dyn_hair_strand_l_bk_02\"),Transform=(Rotation=(X=0,Y=0,Z=0,W=1),Translation=(X=0,Y=0,Z=0),Scale3D=(X=1,Y=1,Z=1))),(Item=(Type=Bone,Name=\"dyn_hair_strand_l_bk_05\"),Transform=(Rotation=(X=0,Y=0,Z=0,W=1),Translation=(X=0,Y=0,Z=0),Scale ...(truncated)"
    ItemToModify.1.Transform.Rotation=&RigVMFunction_MathQuaternionFromEuler(Euler=(X=0,Y=0,Z=0)).Result
    ItemToModify.0.Transform.Rotation=&RigVMFunction_MathQuaternionFromEuler(Euler=(X=0,Y=-10,Z=0)).Result
    Weight=@n5.OutputParam
   )
   AlphaInterp = &RigVMFunction_AlphaInterp(InterpSpeedDecreasing=2, InterpSpeedIncreasing=10, Scale=1, bClampResult=False, Value=RigVMVariableNode() {variable=Target}.Value)
   > &RigUnit_ModifyTransforms(
    ItemToModify=((Item=(Type=Bone,Name="dyn_ponytail_08"),Transform=(Rotation=(X=0,Y=0,Z=0,W=1),Translation=(X=0,Y=3,Z=0),Scale3D=(X=1,Y=1,Z=1))))
    Weight=@AlphaInterp.Result
    ItemToModify.0.Transform.Rotation=&RigVMFunction_MathQuaternionFromEuler(Euler=(X=0,Y=0,Z=0), RotationOrder=XYZ).Result
   )
   > &RigUnit_ModifyTransforms(
    ItemToModify=((Item=(Type=Bone,Name="dyn_ponytail_07"),Transform=(Rotation=(X=0,Y=0,Z=0,W=1),Translation=(X=0,Y=2,Z=0),Scale3D=(X=1,Y=1,Z=1))))
    Weight=@AlphaInterp.Result
    ItemToModify.0.Transform.Rotation=&RigVMFunction_MathQuaternionFromEuler(Euler=(X=0,Y=0,Z=0), RotationOrder=XYZ).Result
   )
   > &RigUnit_ModifyTransforms(
    ItemToModify=((Item=(Type=Bone,Name="dyn_ponytail_05"),Transform=(Rotation=(X=0,Y=0,Z=0,W=1),Translation=(X=0,Y=0,Z=1.5),Scale3D=(X=1,Y=1,Z=1))))
    Weight=@AlphaInterp.Result
    ItemToModify.0.Transform.Rotation=&RigVMFunction_MathQuaternionFromEuler(Euler=(X=0,Y=0,Z=0), RotationOrder=XYZ).Result
   )
   > &RigUnit_ModifyTransforms(
    ItemToModify=((Item=(Type=Bone,Name="dyn_ponytail_03"),Transform=(Rotation=(X=0,Y=0,Z=0,W=1),Translation=(X=0,Y=0,Z=0),Scale3D=(X=1,Y=1,Z=1))))
    Weight=@AlphaInterp.Result
    ItemToModify.0.Transform.Rotation=&RigVMFunction_MathQuaternionFromEuler(Euler=(X=0,Y=0,Z=0), RotationOrder=XYZ).Result
   )
   > n7 = &RigUnit_SphericalPoseReader(
    ActiveRegionScaleFactors=(PositiveWidth=0.4,NegativeWidth=0.3,PositiveHeight=0.3,NegativeHeight=0.4)
    ActiveRegionSize=0.2
    Debug=(bDrawDebug=false,bDraw2D=false,bDrawLocalAxes=false,DebugScale=25,DebugSegments=20,DebugThickness=0.25)
    DriverAxis=(X=-1,Y=0,Z=0)
    DriverItem=(Type=Bone,Name="upperarm_r")
    FalloffRegionScaleFactors=(PositiveWidth=0.05,NegativeWidth=0.1,PositiveHeight=0.2,NegativeHeight=0.2)
    FalloffSize=0.45
    OptionalParentItem=(Type=Bone,Name="neck_01")
    RotationOffset=(X=-70,Y=-20,Z=0)
   )
   > &RigUnit_ModifyTransforms(
    ItemToModify="((Item=(Type=Bone,Name=\"dyn_hair_strand_r_bk_01\"),Transform=(Rotation=(X=0,Y=0,Z=0,W=1),Translation=(X=0,Y=0,Z=0),Scale3D=(X=1,Y=1,Z=1))),(Item=(Type=Bone,Name=\"dyn_hair_strand_r_bk_06\"),Transform=(Rotation=(X=0,Y=0,Z=0,W=1),Translation=(X=0,Y=0,Z=0),Scale ...(truncated)"
    ItemToModify.1.Transform.Rotation=&RigVMFunction_MathQuaternionFromEuler(Euler=(X=0,Y=0,Z=0)).Result
    ItemToModify.0.Transform.Rotation=&RigVMFunction_MathQuaternionFromEuler(Euler=(X=0,Y=0,Z=0)).Result
    Weight=@n7.OutputParam
   )
   > n8 = &RigUnit_ModifyTransforms(
    ItemToModify=((Item=(Type=Bone,Name="dyn_hair_strand_r_01"),Transform=(Rotation=(X=0,Y=0,Z=0,W=1),Translation=(X=0,Y=0,Z=0),Scale3D=(X=1,Y=1,Z=1))))
    ItemToModify.0.Transform.Rotation=&RigVMFunction_MathQuaternionFromEuler(Euler=(X=0,Y=0,Z=10)).Result
    Weight=@n7.OutputParam
   )
   Branch_RigVMUnitNode = &RigVMFunction_ControlFlowBranch(ExecuteContext=@n8.ExecutePin, Condition=RigVMVariableNode() {variable=Emote}.Value)
   Branch_1_RigVMUnitNode = &RigVMFunction_ControlFlowBranch(ExecuteContext=@Branch_RigVMUnitNode.False, Condition=RigVMVariableNode() {variable=Skydive}.Value)
   &RigUnit_ModifyTransforms(
    ItemToModify="((Item=(Type=Bone,Name=\"dyn_ponytail_06\"),Transform=(Rotation=(X=0,Y=0,Z=0,W=1),Translation=(X=0,Y=-2,Z=0),Scale3D=(X=1,Y=1,Z=1))),(Item=(Type=Bone,Name=\"dyn_hair_strand_r_bk_06\"),Transform=(Rotation=(X=0,Y=0,Z=0,W=1),Translation=(X=0,Y=0,Z=0),Scale3D=(X=1 ...(truncated)"
    Weight=1
    ExecutePin=@Branch_1_RigVMUnitNode.True
   )
   &RigUnit_ModifyTransforms(
    ItemToModify=((Item=(Type=Bone,Name="dyn_hair_l_05"),Transform=(Rotation=(X=0,Y=0,Z=0,W=1),Translation=(X=0,Y=0,Z=0),Scale3D=(X=1,Y=1,Z=1))))
    Weight=1
    ExecutePin=@Branch_RigVMUnitNode.True
    ItemToModify.0.Transform.Rotation=&RigVMFunction_MathQuaternionFromEuler(Euler=(X=0,Y=0,Z=-10)).Result
   )
   n9 = &RigUnit_SphericalPoseReader(
    ActiveRegionScaleFactors=(PositiveWidth=0.7,NegativeWidth=0.2,PositiveHeight=0.7,NegativeHeight=0.5)
    ActiveRegionSize=0.2
    Debug=(bDrawDebug=True,DebugScale=25,DebugSegments=20,DebugThickness=0.25)
    DriverAxis=(X=1,Y=0,Z=0)
    DriverItem=(Type=Bone,Name="head")
    ExecutePin=()
    FalloffRegionScaleFactors=(PositiveWidth=0.4,NegativeWidth=0.1,PositiveHeight=0.4,NegativeHeight=0.2)
    FalloffSize=0.45
    OptionalParentItem=(Type=Bone,Name="neck_01")
    RotationOffset=(X=-90,Y=0,Z=110)
   )
   > &RigUnit_ModifyTransforms(
    ItemToModify=((Item=(Type=Bone,Name="dyn_ponytail_08"),Transform=(Rotation=(X=0,Y=0,Z=0,W=1),Translation=(X=0,Y=0,Z=0),Scale3D=(X=1,Y=1,Z=1))))
    ItemToModify.0.Transform.Rotation=&RigVMFunction_MathQuaternionFromEuler(Euler=(X=0,Y=0,Z=-50)).Result
    Weight=@n9.OutputParam
   )
   &RigVMFunction_MathQuaternionFromEuler(Euler=(X=0,Y=30,Z=0))
  ]}
]
hierarchy [
 root {loc=<0 0 0>, rot=<0 0 0 1>}
 pelvis {parent=root, loc=<0 -0.5301 98.0571>, rot=<0.0127 0.707 -0.0127 -0.707>}
 spine_01 {parent=pelvis, loc=<6.1208 -0.9328 0>, rot=<0 0 0.034 -0.9994>}
 spine_02 {parent=spine_01, loc=<6.5996 -0.2095 0>, rot=<0 0 0.0238 -0.9997>}
 spine_03 {parent=spine_02, loc=<6.4914 -0.1237 0>, rot=<0 0 0.0149 -0.9999>}
 spine_04 {parent=spine_03, loc=<6.9598 0.0794 0>, rot=<0 0 -0.0132 -0.9999>}
 spine_05 {parent=spine_04, loc=<6.9987 0 0>, rot=<0 0 -0.0015 -1>}
 neck_01 {parent=spine_05, loc=<11.3138 0.1529 0>, rot=<0 0 -0.2003 -0.9797>}
 neck_02 {parent=neck_01, loc=<4.5627 0 0>, rot=<0 0 0.0504 -0.9987>}
 head {parent=neck_02, loc=<5.3914 0 0>, rot=<0 0 0.115 -0.9934>}
 dyn_hair_bk_MSTR {parent=head, loc=<13.1755 -7.2027 -0.2063>, rot=<0.0081 -0.0236 0.3173 -0.948>}
 dyn_ponytail_01 {parent=dyn_hair_bk_MSTR, loc=<0.7538 -6.3277 -2.42>, rot=<0.4785 -0.8721 -0.0155 -0.1009>}
 dyn_ponytail_02 {parent=dyn_ponytail_01, loc=<2.0858 0 0>, rot=<0.0219 -0.1026 -0.1377 -0.9849>}
 dyn_ponytail_03 {parent=dyn_ponytail_02, loc=<2.314 0 0>, rot=<0.001 0.0156 -0.1354 -0.9907>}
 dyn_ponytail_04 {parent=dyn_ponytail_03, loc=<3.6561 0 0>, rot=<0.0148 0.0537 -0.069 -0.9961>}
 dyn_ponytail_05 {parent=dyn_ponytail_04, loc=<3.5697 0 0>, rot=<0.3745 0.1526 -0.0096 -0.9145>}
 dyn_ponytail_06 {parent=dyn_ponytail_05, loc=<3.5958 0 0>, rot=<0.8619 0.1004 0.1568 -0.4716>}
 dyn_ponytail_07 {parent=dyn_ponytail_06, loc=<5.2811 -0.5267 0.3755>, rot=<-0.0699 0.1429 0.1211 -0.9798>}
 dyn_ponytail_08 {parent=dyn_ponytail_07, loc=<5.7873 0.2031 0.2241>, rot=<-0.1644 0.0572 -0.1328 -0.9757>}
 dyn_ponytail_09 {parent=dyn_ponytail_08, loc=<5.572 0.5496 0.2314>, rot=<-0.0851 0.0301 -0.1175 -0.989>}
 dyn_ponytail_010 {parent=dyn_ponytail_09, loc=<3.2966 0 0>, rot=<0.0286 -0.0507 -0.0122 -0.9982>}
 dyn_ponytail_011 {parent=dyn_ponytail_010, loc=<4.464 0.1187 -0.0821>, rot=<-0.0071 -0.0259 -0.0657 -0.9975>}
 dyn_ponytail_012 {parent=dyn_ponytail_011, loc=<4.0372 0 0>, rot=<-0.0041 -0.0524 -0.1083 -0.9927>}
 dyn_ponytail_013 {parent=dyn_ponytail_012, loc=<2.8238 0.196 -0.1606>, rot=<0.0016 -0.1117 -0.1368 -0.9843>}
 dyn_hair_l_bk_01 {parent=dyn_hair_bk_MSTR, loc=<5.8353 -5.7865 -3.0274>, rot=<0.5101 -0.8175 -0.2273 0.1411>}
 dyn_hair_l_bk_02 {parent=dyn_hair_l_bk_01, loc=<2.8775 0 0>, rot=<0.0079 0.0562 -0.141 -0.9884>}
 dyn_hair_l_bk_03 {parent=dyn_hair_l_bk_02, loc=<3.2813 0 0>, rot=<-0.0118 0.0999 0.0177 -0.9948>}
 dyn_hair_l_bk_04 {parent=dyn_hair_l_bk_03, loc=<3.5121 0 0>, rot=<-0.006 0.0611 0.0343 -0.9975>}
 dyn_hair_l_bk_05 {parent=dyn_hair_l_bk_04, loc=<2.0328 0 0>, rot=<-0.0118 0.0578 0.0848 -0.9946>}
 dyn_hair_l_bk_06 {parent=dyn_hair_l_bk_05, loc=<1.7518 0 0>, rot=<-0.0156 0.0212 0.0907 -0.9955>}
 dyn_hair_r_01 {parent=dyn_hair_bk_MSTR, loc=<2.3008 -9.2947 2.1007>, rot=<0.2612 -0.9645 -0.0146 -0.0355>}
 dyn_hair_r_02 {parent=dyn_hair_r_01, loc=<2.0861 0 0>, rot=<-0.0021 -0.0091 -0.0996 -0.995>}
 dyn_hair_r_03 {parent=dyn_hair_r_02, loc=<3.4481 0 0>, rot=<-0.0049 -0.0136 -0.0215 -0.9997>}
 dyn_hair_r_04 {parent=dyn_hair_r_03, loc=<2.733 0 0>, rot=<0.0038 0.0121 0.0632 -0.9979>}
 dyn_hair_r_05 {parent=dyn_hair_r_04, loc=<2.8626 0 0>, rot=<0.0073 0.0406 0.0641 -0.9971>}
 dyn_hair_r_fr_01 {parent=dyn_hair_bk_MSTR, loc=<2.3143 -4.5639 2.9912>, rot=<0.6502 -0.7561 -0.0042 0.0749>}
 dyn_hair_r_fr_02 {parent=dyn_hair_r_fr_01, loc=<1.4814 0 0>, rot=<-0.0439 0.0691 -0.1985 -0.9767>}
 dyn_hair_r_fr_03 {parent=dyn_hair_r_fr_02, loc=<1.126 0 0>, rot=<0.0001 -0.0006 -0.1359 -0.9907>}
 dyn_hair_r_fr_04 {parent=dyn_hair_r_fr_03, loc=<1.1281 0 0>, rot=<-0.0009 -0.025 -0.1267 -0.9916>}
 dyn_hair_r_fr_05 {parent=dyn_hair_r_fr_04, loc=<1.9624 0 0>, rot=<-0.0193 -0.0681 -0.1143 -0.9909>}
 dyn_hair_r_fr_06 {parent=dyn_hair_r_fr_05, loc=<2.1592 0 0>, rot=<-0.0094 -0.0211 -0.0292 -0.9993>}
 dyn_hair_bk_02 {parent=dyn_hair_bk_MSTR, loc=<5.9852 -7.9609 -1.5864>, rot=<0.4407 -0.8972 -0.0055 -0.0274>}
 dyn_hair_bk_03 {parent=dyn_hair_bk_02, loc=<2.9314 0 0>, rot=<0.0073 -0.0362 -0.0609 -0.9975>}
 dyn_hair_bk_04 {parent=dyn_hair_bk_03, loc=<3.4754 0 0>, rot=<-0.0075 0.035 0.0736 -0.9966>}
 dyn_hair_bk_05 {parent=dyn_hair_bk_04, loc=<2.1169 0 0>, rot=<0.0075 -0.019 0.0906 -0.9957>}
 dyn_hair_r_bk_05 {parent=dyn_hair_bk_MSTR, loc=<4.5759 -9.5268 0.7151>, rot=<0.3557 -0.9334 -0.0139 -0.0465>}
 dyn_hair_r_bk_06 {parent=dyn_hair_r_bk_05, loc=<2.5128 0 0>, rot=<0 -0.0303 -0.0722 -0.9969>}
 dyn_hair_r_bk_07 {parent=dyn_hair_r_bk_06, loc=<1.8963 0 0>, rot=<-0.0003 -0.0026 -0.0388 -0.9992>}
 dyn_hair_r_bk_08 {parent=dyn_hair_r_bk_07, loc=<1.5787 0 0>, rot=<0.0076 0.0419 -0.0314 -0.9986>}
 dyn_hair_r_bk_09 {parent=dyn_hair_r_bk_08, loc=<1.8144 0 0>, rot=<0.0117 0.0561 0.0056 -0.9983>}
 dyn_hair_l_05 {parent=dyn_hair_bk_MSTR, loc=<0.8894 -8.5782 -5.334>, rot=<0.2892 -0.8953 -0.3144 0.1259>}
 dyn_hair_l_06 {parent=dyn_hair_l_05, loc=<2.7407 0 0>, rot=<0.0037 0.0764 -0.0035 -0.9971>}
 dyn_hair_l_07 {parent=dyn_hair_l_06, loc=<3.0764 0 0>, rot=<0.0045 0.0275 0.0312 -0.9991>}
 dyn_hair_l_08 {parent=dyn_hair_l_07, loc=<2.6693 0 0>, rot=<0.0008 -0.0268 0.0703 -0.9972>}
 dyn_hair_l_09 {parent=dyn_hair_l_08, loc=<2.7046 0 0>, rot=<-0.0052 -0.0225 0.0926 -0.9954>}
 dyn_hair_l_010 {parent=dyn_hair_l_09, loc=<2.0224 0 0>, rot=<-0.0178 -0.0049 0.0995 -0.9949>}
 dyn_hair_l_011 {parent=dyn_hair_l_010, loc=<1.9859 0 0>, rot=<-0.0222 0.0249 0.0533 -0.998>}
 dyn_hair_l_fr_02 {parent=dyn_hair_bk_MSTR, loc=<1.5048 -3.2852 -5.0309>, rot=<0.5153 -0.7119 -0.4573 0.1363>}
 dyn_hair_l_fr_03 {parent=dyn_hair_l_fr_02, loc=<1.8971 0 0>, rot=<0.028 0.0749 -0.1392 -0.987>}
 dyn_hair_l_fr_04 {parent=dyn_hair_l_fr_03, loc=<2.4746 0 0>, rot=<0.0017 0.0917 -0.0853 -0.9921>}
 dyn_hair_l_fr_05 {parent=dyn_hair_l_fr_04, loc=<3.0287 0 0>, rot=<-0.0031 0.121 -0.0772 -0.9896>}
 dyn_hair_l_fr_06 {parent=dyn_hair_l_fr_05, loc=<2.8191 0 0>, rot=<-0.0009 0.0057 -0.0055 -1>}
 dyn_hair_strand_l_01 {parent=head, loc=<8.8999 6.1201 -6.7399>, rot=<-0.1624 0.5455 0.8217 -0.0294>}
 dyn_hair_strand_l_02 {parent=dyn_hair_strand_l_01, loc=<2.9436 0 0>, rot=<0.0212 0.0421 -0.0868 -0.9951>}
 dyn_hair_strand_l_03 {parent=dyn_hair_strand_l_02, loc=<2.1449 0 0>, rot=<0.0046 0.0404 -0.0094 -0.9991>}
 dyn_hair_strand_l_04 {parent=dyn_hair_strand_l_03, loc=<2.2467 0 0>, rot=<0.0009 0.0212 0.0015 -0.9998>}
 dyn_hair_strand_l_05 {parent=dyn_hair_strand_l_04, loc=<2.0481 0 0>, rot=<0.0002 0.0152 0.0049 -0.9999>}
 dyn_hair_strand_l_06 {parent=dyn_hair_strand_l_05, loc=<2.991 0 0>, rot=<0.0002 0.046 0.0149 -0.9988>}
 dyn_hair_strand_l_07 {parent=dyn_hair_strand_l_06, loc=<3.3442 0 0>, rot=<-0.0008 0.0384 0.0011 -0.9993>}
 dyn_hair_strand_l_08 {parent=dyn_hair_strand_l_07, loc=<2.437 0 0>, rot=<-0.004 0.0368 -0.0211 -0.9991>}
 dyn_hair_strand_l_09 {parent=dyn_hair_strand_l_08, loc=<1.9394 0 0>, rot=<-0.0116 0.0367 -0.0711 -0.9967>}
 dyn_hair_strand_l_fr_01 {parent=head, loc=<8.3421 6.9758 -5.4804>, rot=<-0.162 0.5143 0.8416 0.0296>}
 dyn_hair_strand_l_fr_02 {parent=dyn_hair_strand_l_fr_01, loc=<1.8433 0 0>, rot=<0.0092 0.032 -0.0732 -0.9968>}
 dyn_hair_strand_l_fr_03 {parent=dyn_hair_strand_l_fr_02, loc=<2.2725 0 0>, rot=<0.0095 0.0428 -0.0695 -0.9966>}
 dyn_hair_strand_l_fr_04 {parent=dyn_hair_strand_l_fr_03, loc=<2.1812 0 0>, rot=<0.0044 0.0415 -0.0333 -0.9986>}
 dyn_hair_fr_l_01 {parent=head, loc=<10.1436 8.8326 -3.4718>, rot=<-0.2072 0.2812 0.9249 0.1499>}
 dyn_hair_fr_l_02 {parent=dyn_hair_fr_l_01, loc=<1.2645 0 0>, rot=<-0.0078 0.0221 -0.106 -0.9941>}
 dyn_hair_fr_l_03 {parent=dyn_hair_fr_l_02, loc=<1.4127 0 0>, rot=<0.0005 0.0032 -0.0506 -0.9987>}
 dyn_hair_fr_l_04 {parent=dyn_hair_fr_l_03, loc=<1.2715 0 0>, rot=<0.0045 0.0289 -0.0547 -0.9981>}
 dyn_hair_c_l_01 {parent=head, loc=<11.6031 8.5929 -1.205>, rot=<0.1265 -0.2482 -0.8139 -0.5099>}
 dyn_hair_c_l_02 {parent=dyn_hair_c_l_01, loc=<1.279 0 0>, rot=<0.0295 -0.0163 -0.2525 -0.967>}
 dyn_hair_c_l_03 {parent=dyn_hair_c_l_02, loc=<2.0226 0 0>, rot=<-0.0373 -0.0046 -0.1789 -0.9831>}
 dyn_hair_c_l_04 {parent=dyn_hair_c_l_03, loc=<1.744 0 0>, rot=<-0.0072 0.0046 -0.0813 -0.9967>}
 dyn_hair_c_r_01 {parent=head, loc=<11.7376 9.1688 0.7677>, rot=<0.0947 -0.2127 0.9086 0.3466>}
 dyn_hair_c_r_02 {parent=dyn_hair_c_r_01, loc=<1.7368 0 0>, rot=<0.0409 0.0196 -0.1651 -0.9852>}
 dyn_hair_c_r_03 {parent=dyn_hair_c_r_02, loc=<1.3831 0 0>, rot=<0.0086 0.0105 -0.0689 -0.9975>}
 dyn_hair_c_r_04 {parent=dyn_hair_c_r_03, loc=<2.0155 0 0>, rot=<0.0038 0.0276 -0.0976 -0.9948>}
 dyn_hair_fr_r_01 {parent=head, loc=<10.4461 9.0048 3.1005>, rot=<0.1544 -0.2972 0.9102 0.2436>}
 dyn_hair_fr_r_02 {parent=dyn_hair_fr_r_01, loc=<1.3716 0 0>, rot=<0.0299 -0.0278 -0.1154 -0.9925>}
 dyn_hair_fr_r_03 {parent=dyn_hair_fr_r_02, loc=<1.4916 0 0>, rot=<0.0132 -0.0117 -0.127 -0.9917>}
 dyn_hair_side_r_01 {parent=head, loc=<8.998 8.0572 4.7045>, rot=<0.1425 -0.5054 0.8362 0.1584>}
 dyn_hair_side_r_02 {parent=dyn_hair_side_r_01, loc=<1.1935 0 0>, rot=<0.0119 -0.0355 -0.0919 -0.9951>}
 dyn_hair_side_r_03 {parent=dyn_hair_side_r_02, loc=<1.3444 0 0>, rot=<0.0037 0.0214 -0.0846 -0.9962>}
 dyn_hair_strand_r_fr_01 {parent=head, loc=<9.4638 7.1484 5.1578>, rot=<0.2287 -0.4793 0.8279 0.1804>}
 dyn_hair_strand_r_fr_02 {parent=dyn_hair_strand_r_fr_01, loc=<1.8706 0.0612 -0.0126>, rot=<0.0038 -0.0465 -0.2293 -0.9722>}
 dyn_hair_strand_r_fr_03 {parent=dyn_hair_strand_r_fr_02, loc=<2.1348 -0.1413 0.0124>, rot=<-0.0061 -0.0074 -0.0791 -0.9968>}
 dyn_hair_strand_r_fr_04 {parent=dyn_hair_strand_r_fr_03, loc=<2.3163 -0.0138 0.0062>, rot=<-0.0024 -0.0086 -0.019 -0.9998>}
 dyn_hair_strand_r_fr_05 {parent=dyn_hair_strand_r_fr_04, loc=<1.8105 -0.0083 -0.0024>, rot=<-0.0029 0.0081 -0.0283 -0.9996>}
 dyn_hair_strand_r_01 {parent=head, loc=<4.0423 5.012 6.9033>, rot=<0.0755 -0.4944 0.8655 -0.0256>}
 dyn_hair_strand_r_02 {parent=dyn_hair_strand_r_01, loc=<3.2025 0.0478 0.0564>, rot=<-0.0033 -0.0595 -0.0195 -0.998>}
 dyn_hair_strand_r_03 {parent=dyn_hair_strand_r_02, loc=<4.0033 0 0>, rot=<0.0004 -0.0627 0.0241 -0.9977>}
 dyn_hair_strand_r_04 {parent=dyn_hair_strand_r_03, loc=<2.2303 0 0>, rot=<0.0017 -0.0281 0.0014 -0.9996>}
 dyn_hair_strand_r_05 {parent=dyn_hair_strand_r_04, loc=<2.1092 0.0044 0.0171>, rot=<0.0061 -0.0844 0.0212 -0.9962>}
 dyn_hair_strand_r_bk_01 {parent=head, loc=<6.3035 3.719 7.5172>, rot=<0.1158 -0.6199 0.7695 -0.1011>}
 dyn_hair_strand_r_bk_02 {parent=dyn_hair_strand_r_bk_01, loc=<2.4282 0 0>, rot=<-0.0099 -0.0579 -0.0273 -0.9979>}
 dyn_hair_strand_r_bk_03 {parent=dyn_hair_strand_r_bk_02, loc=<2.1774 0 0>, rot=<-0.0072 -0.0376 -0.0357 -0.9986>}
 dyn_hair_strand_r_bk_04 {parent=dyn_hair_strand_r_bk_03, loc=<2.5087 0 0>, rot=<0.0026 -0.0402 0.0389 -0.9984>}
 dyn_hair_strand_r_bk_05 {parent=dyn_hair_strand_r_bk_04, loc=<3.4382 0 0>, rot=<0.0001 -0.0486 0.0026 -0.9988>}
 dyn_hair_strand_r_bk_06 {parent=dyn_hair_strand_r_bk_05, loc=<3.1118 0 0>, rot=<0.0042 -0.0524 -0.0316 -0.9981>}
 dyn_hair_strand_r_bk_07 {parent=dyn_hair_strand_r_bk_06, loc=<3.0345 0 0>, rot=<0.0098 -0.0321 -0.0542 -0.998>}
 dyn_hair_strand_r_bk_08 {parent=dyn_hair_strand_r_bk_07, loc=<2.0875 0 0>, rot=<0.0151 -0.0493 -0.0592 -0.9969>}
 dyn_earring_l {parent=head, loc=<-0.1457 -0.2162 -6.6178>, rot=<-0.0101 -0.9998 0.0003 0.0184>}
 dyn_earring_r {parent=head, loc=<-0.2284 -0.3594 6.6522>, rot=<-0.0182 -0.9997 0.0005 0.0182>}
 dyn_hair_strand_l_bk_02 {parent=head, loc=<4.6145 2.8846 -7.5491>, rot=<-0.0767 0.6886 0.7187 -0.0585>}
 dyn_hair_strand_l_bk_03 {parent=dyn_hair_strand_l_bk_02, loc=<2.7223 0 0>, rot=<0.0061 0.0557 -0.0445 -0.9974>}
 dyn_hair_strand_l_bk_04 {parent=dyn_hair_strand_l_bk_03, loc=<4.0699 0 0>, rot=<0.0004 0.0467 -0.0126 -0.9988>}
 dyn_hair_strand_l_bk_05 {parent=dyn_hair_strand_l_bk_04, loc=<3.6556 0 0>, rot=<0.0001 0.0531 0.0036 -0.9986>}
 dyn_hair_strand_l_bk_06 {parent=dyn_hair_strand_l_bk_05, loc=<3.2383 0 0>, rot=<-0.0099 0.0378 -0.0604 -0.9974>}
 dyn_hair_strand_l_bk_07 {parent=dyn_hair_strand_l_bk_06, loc=<2.3488 0 0>, rot=<-0.0187 0.0375 -0.0791 -0.996>}
 FACIAL_C_FacialRoot {parent=head, loc=<1.4167 0.1416 0>, rot=<0.5021 -0.4979 -0.5021 -0.4979>}
 FACIAL_L_ForeheadIn {parent=FACIAL_C_FacialRoot, loc=<2.1591 -3.499 8.1668>, rot=<-0.0609 -0.0696 0.0043 -0.9957>}
 FACIAL_L_ForeheadMid {parent=FACIAL_C_FacialRoot, loc=<3.8171 -3.7915 7.4869>, rot=<-0.0712 -0.1757 -0.0206 -0.9816>}
 FACIAL_L_ForeheadOut {parent=FACIAL_C_FacialRoot, loc=<5.0703 -3.3965 6.459>, rot=<-0.1215 -0.2557 -0.0792 -0.9558>}
 FACIAL_L_EyelidUpperFurrow {parent=FACIAL_C_FacialRoot, loc=<3.5744 -2.5784 8.148>, rot=<-0.0301 -0.1509 -0.0797 -0.9849>}
 FACIAL_L_EyesackUpper {parent=FACIAL_C_FacialRoot, loc=<3.6801 -2.794 8.2201>, rot=<0.0487 -0.0843 -0.0646 -0.9932>}
 clavicle_l {parent=spine_05, loc=<5.4831 -0.1053 -3.2946>, rot=<0.0465 -0.7289 -0.009 -0.6829>}
 upperarm_l {parent=clavicle_l, loc=<9.8658 -1.2136 -0.4105>, rot=<0.0164 -0.3245 -0.0269 -0.9454>}
 upperarm_twist_01_l {parent=upperarm_l, loc=<7.2135 -0.1134 0>, rot=<0 0 0 1>}
 clavicle_r {parent=spine_05, loc=<5.4831 -0.1053 3.2946>, rot=<0.7289 0.0465 -0.6829 0.009>}
 upperarm_r {parent=clavicle_r, loc=<-9.8658 1.2136 0.4105>, rot=<0.0164 -0.3245 -0.0269 -0.9454>}
]
variables [
 Emote {type=bool, default=False}
 Melee {type=double, default=0}
 Skydive {type=bool, default=False}
 Target {type=double, default=0}
]
