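# Physics asset for the lower ponytail of the F_MED_IvyCross face accessory, setting up the dyn_ponytail_06 to 013 bones for physical motion.
# A chain of small capsules is jointed to spine_05, with tight twist limits throughout and swing limits that generally widen toward the tip.
# Larger capsules and a shoulder sphere cover the upper spine and left clavicle and upper arm only, joined by free angular constraints, apparently as colliders for the hair.
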PhysicsAsset /BRCosmetics/Characters/Player/Female/Medium/Bodies/F_MED_IvyCross/Meshes/Parts/F_MED_IvyCross_FaceAcc_Physics_Ponytail_tip
bodies {
 clavicle_l [{
   shapeType=Capsule
   length=23.1084
   radius=5.345
   shapeName=clavicle_l_capsule
   center=<0.1081 0 0.0148>
   rotation=<rot 84.375 -0.0002 -0.0003>
  }]
 dyn_ponytail_010 [{shapeType=Capsule, length=3, radius=1, shapeName=dyn_ponytail_010_capsule, center=<3 0 1>, rotation=<rot 90 0 0>}]
 dyn_ponytail_011 [{shapeType=Capsule, length=3, radius=1.5, shapeName=dyn_ponytail_011_capsule, center=<3 0 1>, rotation=<rot 90 0 0>}]
 dyn_ponytail_012 [{shapeType=Capsule, length=3, radius=1.5, shapeName=dyn_ponytail_012_capsule, center=<3 0 1>, rotation=<rot 90 0 0>}]
 dyn_ponytail_013 [{shapeType=Capsule, length=3, radius=1.5, shapeName=dyn_ponytail_013_capsule, center=<3 0 1>, rotation=<rot 90 0 0>}]
 dyn_ponytail_06 [{shapeType=Capsule, length=3, radius=2, center=<3 -1 0>, rotation=<rot 90 0 0>}]
 dyn_ponytail_07 [{shapeType=Capsule, length=2.9337, radius=1.4286, shapeName=dyn_ponytail_07_capsule, center=<3 0 1>, rotation=<rot 90 0 0>}]
 dyn_ponytail_08 [{shapeType=Capsule, length=2.5714, radius=1.4286, shapeName=dyn_ponytail_08_capsule, center=<3 0 1>, rotation=<rot 90 0 0>}]
 dyn_ponytail_09 [{shapeType=Capsule, length=2.5714, radius=1.4286, shapeName=dyn_ponytail_09_capsule, center=<3 0 1>, rotation=<rot 90 0 0>}]
 spine_03 [{shapeType=Capsule, length=21.226, radius=9.3255, shapeName=spine_03_capsule, center=<-1.8235 -1.033 0>, rotation=<rot 0 -90 0>}]
 spine_04 [{shapeType=Capsule, length=17.0626, radius=9.3255, shapeName=spine_04_capsule, center=<1.8634 0.5618 0>, rotation=<rot 0 -90 0>}]
 spine_05 [{shapeType=Capsule, length=24.1437, radius=7.717, shapeName=spine_05_capsule, center=<0 -2.2755 0>}]
 upperarm_l [
  {shapeType=Sphere, radius=7.7864, shapeName=upperarm_l_sphere, center=<-0.3036 0 5.9724>}
  {shapeType=Capsule, length=13.55, radius=4.3007, shapeName=upperarm_l_capsule, center=<7.598 0 1.9032>, rotation=<rot -84.375 0.0007 -0.0007>}
 ]
}
constraints [
 {bone1Name=dyn_ponytail_06, bone2Name=spine_05, swing1LimitDegrees=5, swing2LimitDegrees=5, twistLimitDegrees=2}
 {
  bone1Name=clavicle_l
  bone2Name=spine_05
  swing1LimitDegrees=45
  swing1Motion=Free
  swing2LimitDegrees=45
  swing2Motion=Free
  twistLimitDegrees=45
  twistMotion=Free
 }
 {
  bone1Name=upperarm_l
  bone2Name=clavicle_l
  swing1LimitDegrees=45
  swing1Motion=Free
  swing2LimitDegrees=45
  swing2Motion=Free
  twistLimitDegrees=45
  twistMotion=Free
 }
 {bone1Name=dyn_ponytail_07, bone2Name=dyn_ponytail_06, swing1LimitDegrees=30, swing2LimitDegrees=10, twistLimitDegrees=2}
 {bone1Name=dyn_ponytail_08, bone2Name=dyn_ponytail_07, swing1LimitDegrees=30, swing2LimitDegrees=15, twistLimitDegrees=2}
 {bone1Name=dyn_ponytail_09, bone2Name=dyn_ponytail_08, swing1LimitDegrees=30, swing2LimitDegrees=15, twistLimitDegrees=2}
 {bone1Name=dyn_ponytail_010, bone2Name=dyn_ponytail_09, swing1LimitDegrees=30, swing2LimitDegrees=30, twistLimitDegrees=2}
 {bone1Name=dyn_ponytail_011, bone2Name=dyn_ponytail_010, swing1LimitDegrees=30, swing2LimitDegrees=30, twistLimitDegrees=2}
 {bone1Name=dyn_ponytail_012, bone2Name=dyn_ponytail_011, swing1LimitDegrees=30, swing2LimitDegrees=30, twistLimitDegrees=2}
 {bone1Name=dyn_ponytail_013, bone2Name=dyn_ponytail_012, swing1LimitDegrees=30, swing2LimitDegrees=30, twistLimitDegrees=2}
 {
  bone1Name=spine_05
  bone2Name=spine_04
  swing1LimitDegrees=45
  swing1Motion=Free
  swing2LimitDegrees=45
  swing2Motion=Free
  twistLimitDegrees=45
  twistMotion=Free
 }
 {
  bone1Name=spine_04
  bone2Name=spine_03
  swing1LimitDegrees=45
  swing1Motion=Free
  swing2LimitDegrees=45
  swing2Motion=Free
  twistLimitDegrees=45
  twistMotion=Free
 }
]
